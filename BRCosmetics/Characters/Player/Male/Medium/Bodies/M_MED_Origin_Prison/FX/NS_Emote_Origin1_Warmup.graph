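# Niagara effect for the warmup of the Origin1 emote on the Origin Prison body, dressing the character in blue-cyan glows.
# BlueFlares pins pulsing sprites and meshes to User.SkeletalMesh points picked by particle ID through an index array.
# Corona adds a burst sprite glow and light, while SubUV_ConeGlowLoop spawns mesh-attached sprites aligned to sampled rotation.
# Four other emitters are disabled: cube trails, chroma sprites, black decals and a second flare.
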NiagaraSystem /BRCosmetics/Characters/Player/Male/Medium/Bodies/M_MED_Origin_Prison/FX/NS_Emote_Origin1_Warmup
dependencies {
 usedDataInterfaces [
  &NiagaraDataInterfaceMeshRendererInfo
  &NiagaraDataInterfaceCurve
  &NiagaraDataInterfaceArrayInt32
  &NiagaraDataInterfaceVector2DCurve
  &NiagaraDataInterfaceColorCurve
  &NiagaraDataInterfaceArrayFloat3
 ]
 usedDynamicInputs [&NormalizeVector]
 usedModules [
  &SystemState
  &self:SystemSpawnScript.SystemScriptSource.SystemScriptGraph.NiagaraNodeAssignment_5.SetVariables_#1
  &InterpolateOverTime
  &self:SystemSpawnScript.SystemScriptSource.SystemScriptGraph.NiagaraNodeAssignment_1.SetVariables_#2
  &self:SystemSpawnScript.SystemScriptSource.SystemScriptGraph.NiagaraNodeAssignment_0.SetVariables_#3
 ]
}
emitters {
 BlackDecals {
  bEnabled=false
  emitterUpdateScript [
   &EmitterState(
    Inactive Response=::NewEnumerator1
    Life Cycle Mode=::NewEnumerator1
    Loop Behavior=::NewEnumerator1
    Loop Delay=0.15
    Loop Duration=3
    UseLoopDelay=true
   )
   &SpawnBurst_Instantaneous(
    Spawn Count=20
    Spawn Probability=&MakeCustomFloatFromBool(Bool=$User.bIsFrontEndPreview, True Float=1, False Float=0)
    Spawn Time=0.01
    Use Spawn Probability=true
   ) {enabled=false}
   &SpawnRate(SpawnRate=15, Use Spawn Probability=false)
  ]
  particleSpawnScript [/Niagara/Modules/Spawn/Initialization/InitializeParticle(
    Color=<rgba 0 0 0 1>
    Lifetime=4
    Module.Write Color=true
    Module.Write Lifetime=true
    Module.Write Mass=true
    Module.Write Material Random=true
    Module.Write Position=false
    Module.Write Scale=false
    Module.Write SpriteRotation=false
    Module.Write SpriteSize=true
    Sprite Size=&Vector2DFromFloat(Value=50)
   )]
  particleUpdateScript [
   &ParticleState(Kill Particles When Lifetime Has Elapsed=false)
   &Color(Color=&MultiplyLinearColor(
     Linear Color=<rgba 0 0 0 1>
     Scale Factor=&MakeFloat4FromVectorAndFloat("Vector (RGB)"=&VectorFromFloat(Value=3), "Float (Alpha)"=1)
    ))
   &ScaleColor(
    Color Value To Scale=$Particles.Color
    Scale Alpha=&FloatFromCurve(
     FloatCurve=DataInterface() {propertyValues {Curve {keys [{arriveTangent=7.9503, interpMode=RCIM_Cubic, time=0, value=0}, <at 0.8 1.25 RCIM_Cubic>, <at 1 0>]}}}
     Scale Curve=20
    )
    Scale RGB=&VectorFromFloat(Value=1)
    ScaleA=true
    ScaleRGB=true
   )
   &SkeletalMeshLocation(
    Bone / Socket Index=&ReturnParticleIDAttributes()
    Bone Sampling Mode=::NewEnumerator4
    Inherit Velocity=false
    MaskSpawnGroup=false
    Skeletal Mesh=DataInterface() {propertyValues {
      DefaultMesh=null
      ExcludeBoneName=root
      OptionalComponentClass=null
      PreviewMesh=/Game/Characters/Player/Female/Medium/Bodies/F_MED_Scholar/Meshes/F_MED_Scholar
      SkinningMode=SkinOnTheFly
      SoftSourceActor=null
      SourceMode=Default
      UvSetIndex=0
      WholeMeshLOD=-1
      bExcludeBone=false
      bReadDeformedGeometry=true
      bRequireCurrentFrameData=true
      ComponentTags []
      FilteredBones [
       pelvis
       spine_02
       upperarm_l
       lowerarm_l
       thigh_twist_01_l
       upperarm_r
       lowerarm_r
       thigh_twist_01_r
       neck_01
       thigh_l
       calf_l
       spine_05
       thigh_r
       calf_r
       foot_r
       foot_l
       head
       spine_04
       hand_r
       hand_l
      ]
      FilteredSockets []
      MeshUserParameter {parameter {name=None, typeDefHandle {registeredTypeIndex=1}, varData []}}
      SamplingRegions []
     }}
   )
   &CameraOffset(Camera Offset Amount=25)
   &DynamicMaterialParameters(
    Index 0 Param 1=&FloatFromCurve(FloatCurve=DataInterface() {propertyValues {Curve {keys [
         <at 0 -0.5 RCIM_Cubic>
         {arriveTangent=10, interpMode=RCIM_Cubic, leaveTangent=10, time=0.0692, value=0.2738}
         {arriveTangent=5.0777, interpMode=RCIM_Cubic, leaveTangent=5.0777, time=0.15, value=1}
         {interpMode=RCIM_Cubic, leaveTangent=4.2839, time=1, value=5}
        ]}}})
    Index 0 Param 3=&FloatFromCurve(
     FloatCurve=DataInterface() {propertyValues {Curve {keys [{arriveTangent=2.3217, interpMode=RCIM_Cubic, time=0, value=-0.25}, {interpMode=RCIM_Cubic, leaveTangent=0.1844, time=1, value=1.4}]}}}
     Scale Curve=0.5
    )
    Index 0 Param 4=1
    Module.Param0WriteEnabled=true
    Module.Param1WriteEnabled=true
    Module.Param2WriteEnabled=true
    Module.Param3WriteEnabled=true
   )
  ]
  rendererClasses [&NiagaraSpriteRendererProperties]
  renderers [{rendererClass=&NiagaraSpriteRendererProperties}]
 }
 BlueFlares {
  emitterUpdateScript [
   &EmitterState(Life Cycle Mode=::NewEnumerator1, Loop Behavior=::NewEnumerator1, Loop Duration=0.9, UseLoopDelay=false)
   &SpawnBurst_Instantaneous(Spawn Count=13, Use Spawn Probability=false, UseLoopCountLimit=false) {enabled=false}
   &SpawnRate(SpawnRate=50, Use Spawn Probability=false)
  ]
  particleSpawnScript [
   /Niagara/Modules/Spawn/Initialization/V2/InitializeParticle(
    Color=<rgba 0.4072 0.8963 0.8796 0.5>
    Lifetime=2
    Position Mode=::NewEnumerator0
    Sprite Size Mode=::NewEnumerator3
    Uniform Sprite Size=8
    Write Material Random=true
   )
   &Color(
    Color=&HueShiftLinearColor(Hue=0.95, Saturation=1.5, Value=5, Linear Color=$Particles.Color)
    Scale Alpha=0.1
   )
  ]
  particleUpdateScript [
   &ParticleState()
   &SkeletalMeshLocation(
    Bone / Socket Index=&SelectIntFromArray(
     Direct Array Index=&ReturnParticleIDAttributes()
     Direct Array Mode=::NewEnumerator1
     Int Selection Array=DataInterface() {propertyValues {IntData [0,1,2,3,4,5,6,7,8,9,10,11,12,13]}}
    )
    MaskSpawnGroup=false
    Mesh Sampling Type=::NewEnumerator5
    Skeletal Mesh=$User.SkeletalMesh
    Socket Sampling Mode=::NewEnumerator5
   )
   &CameraOffset(Camera Offset Amount=1) {enabled=false}
   &DynamicMaterialParameters(
    Index 0 Param 1=0.5
    Index 0 Param 2=1
    Module.Param0WriteEnabled=true
    Module.Param1WriteEnabled=true
    Module.Param2WriteEnabled=true
    Module.Param3WriteEnabled=true
   )
   &ScaleSpriteSize(
    Uniform Curve Scale=2
    Uniform Curve Sprite Scale=DataInterface() {propertyValues {Curve {keys [
        {arriveTangent=45.5455, interpMode=RCIM_Cubic, leaveTangent=45.5455, tangentMode=RCTM_User, time=0, value=0}
        {arriveTangent=7.9167, interpMode=RCIM_Cubic, leaveTangent=7.9167, time=0.1, value=4}
        {arriveTangent=-4.0389, interpMode=RCIM_Cubic, leaveTangent=-4.0389, tangentMode=RCTM_User, time=0.3, value=1}
        {arriveTangent=-0.3032, interpMode=RCIM_Cubic, leaveTangent=-0.3032, tangentMode=RCTM_User, time=1, value=0.2}
       ]}}}
   )
  ]
  rendererClasses [&NiagaraSpriteRendererProperties, &NiagaraMeshRendererProperties]
  renderers [
   {rendererClass=&NiagaraSpriteRendererProperties}
   {rendererClass=&NiagaraMeshRendererProperties, rendererIndex=1}
  ]
 }
 BlueFlares001 {
  bEnabled=false
  emitterUpdateScript [
   &EmitterState(Life Cycle Mode=::NewEnumerator1, Loop Duration=1, UseLoopDelay=false)
   &SpawnBurst_Instantaneous(Spawn Count=13, Use Spawn Probability=false, UseLoopCountLimit=false) {enabled=false}
   &SpawnRate(SpawnRate=20, Use Spawn Probability=false)
  ]
  particleSpawnScript [
   &InitializeParticle(
    Color=<rgba 0.4072 0.8963 0.8796 0.1>
    Lifetime=2
    Position Mode=::NewEnumerator0
    Sprite Size Mode=::NewEnumerator3
    Uniform Sprite Size=20
    Write Material Random=true
   )
   &Color(
    Color=&HueShiftLinearColor(Hue=0.95, Saturation=1.5, Value=5, Linear Color=$Particles.Color)
    Scale Alpha=0.1
   )
  ]
  particleUpdateScript [
   &ParticleState()
   &SkeletalMeshLocation(
    Bone / Socket Index=&SelectIntFromArray(
     Direct Array Index=&ReturnParticleIDAttributes()
     Direct Array Mode=::NewEnumerator1
     Int Selection Array=DataInterface() {propertyValues {IntData [0,1,2,3,4,5,6,7,8,9,10,11]}}
    )
    MaskSpawnGroup=false
    Mesh Sampling Type=::NewEnumerator5
    Skeletal Mesh=$User.SkeletalMesh
    Socket Sampling Mode=::NewEnumerator5
   )
   &CameraOffset(Camera Offset Amount=1) {enabled=false}
   &DynamicMaterialParameters(
    Index 0 Param 1=0.5
    Index 0 Param 2=1
    Module.Param0WriteEnabled=true
    Module.Param1WriteEnabled=true
    Module.Param2WriteEnabled=true
    Module.Param3WriteEnabled=true
   )
   &ScaleSpriteSize(
    Uniform Curve Scale=2
    Uniform Curve Sprite Scale=DataInterface() {propertyValues {Curve {keys [<at 0 0 RCIM_Cubic>, {arriveTangent=0.5, interpMode=RCIM_Cubic, leaveTangent=0.5, time=0.3, value=4}, <at 0.5 2.375>, <at 2 1 RCIM_Cubic>]}}}
   )
  ]
  rendererClasses [&NiagaraSpriteRendererProperties, &NiagaraMeshRendererProperties]
  renderers [
   {rendererClass=&NiagaraSpriteRendererProperties}
   {rendererClass=&NiagaraMeshRendererProperties, rendererIndex=1}
  ]
 }
 ChromaSprites {
  bEnabled=false
  emitterUpdateScript [
   &EmitterState(
    Inactive Response=::NewEnumerator1
    Life Cycle Mode=::NewEnumerator1
    Loop Behavior=::NewEnumerator1
    Loop Delay=0.15
    Loop Duration=1
    UseLoopDelay=true
   )
   &SpawnBurst_Instantaneous(
    Spawn Count=20
    Spawn Probability=&MakeCustomFloatFromBool(Bool=$User.bIsFrontEndPreview, True Float=1, False Float=0)
    Spawn Time=0.01
    Use Spawn Probability=true
   ) {enabled=false}
   &SpawnRate(SpawnRate=20, Use Spawn Probability=false)
  ]
  particleSpawnScript [/Niagara/Modules/Spawn/Initialization/InitializeParticle(
    Color=&MakeLinearColorFromVectorAndFloat("Vector (RGB)"=&Multiply_VectorByFloat(Vector=<0 0.2 0.43>, Float=0.025), "Float (Alpha)"=1)
    Lifetime=0.95
    Module.Write Color=true
    Module.Write Lifetime=true
    Module.Write Mass=true
    Module.Write Material Random=true
    Module.Write Position=false
    Module.Write Scale=false
    Module.Write SpriteRotation=false
    Module.Write SpriteSize=true
    Sprite Size=&Vector2DFromFloat(Value=50)
   )]
  particleUpdateScript [
   &ParticleState(Kill Particles When Lifetime Has Elapsed=false)
   &Color(Color=&MultiplyLinearColor(
     Linear Color=&MakeLinearColorFromVectorAndFloat("Vector (RGB)"=&Multiply_VectorByFloat(Vector=<0.2 0.2 0>, Float=0.11), "Float (Alpha)"=1)
     Scale Factor=&MakeFloat4FromVectorAndFloat("Vector (RGB)"=&VectorFromFloat(Value=3), "Float (Alpha)"=1)
    ))
   &ScaleColor(
    Color Value To Scale=$Particles.Color
    Scale Alpha=&FloatFromCurve(FloatCurve=DataInterface() {propertyValues {Curve {keys [{arriveTangent=7.9503, interpMode=RCIM_Cubic, time=0, value=0}, <at 0.125 1.25 RCIM_Cubic>, <at 1 0>]}}})
    Scale RGB=&VectorFromFloat(Value=1)
    ScaleA=true
    ScaleRGB=true
   )
   &SkeletalMeshLocation(
    Bone / Socket Index=&ReturnParticleIDAttributes()
    Bone Sampling Mode=::NewEnumerator4
    Inherit Velocity=false
    MaskSpawnGroup=false
    Skeletal Mesh=DataInterface() {propertyValues {
      DefaultMesh=null
      ExcludeBoneName=root
      OptionalComponentClass=null
      PreviewMesh=&F_MED_Scholar
      SkinningMode=SkinOnTheFly
      SoftSourceActor=null
      SourceMode=Default
      UvSetIndex=0
      WholeMeshLOD=-1
      bExcludeBone=false
      bReadDeformedGeometry=true
      bRequireCurrentFrameData=true
      ComponentTags []
      FilteredBones [
       pelvis
       spine_02
       upperarm_l
       lowerarm_l
       thigh_twist_01_l
       upperarm_r
       lowerarm_r
       thigh_twist_01_r
       neck_01
       thigh_l
       calf_l
       spine_05
       thigh_r
       calf_r
       foot_r
       foot_l
       head
       spine_04
       hand_r
       hand_l
      ]
      FilteredSockets []
      MeshUserParameter {parameter {name=None, typeDefHandle {registeredTypeIndex=1}, varData []}}
      SamplingRegions []
     }}
   )
   &CameraOffset(Camera Offset Amount=25) {enabled=false}
   &DynamicMaterialParameters(
    Index 0 Param 1=&FloatFromCurve(FloatCurve=DataInterface() {propertyValues {Curve {keys [
         <at 0 -0.5 RCIM_Cubic>
         {arriveTangent=10, interpMode=RCIM_Cubic, leaveTangent=10, time=0.0692, value=0.2738}
         {arriveTangent=5.0777, interpMode=RCIM_Cubic, leaveTangent=5.0777, time=0.15, value=1}
         {interpMode=RCIM_Cubic, leaveTangent=4.2839, time=1, value=5}
        ]}}})
    Index 0 Param 3=&FloatFromCurve(
     FloatCurve=DataInterface() {propertyValues {Curve {keys [{arriveTangent=2.3217, interpMode=RCIM_Cubic, time=0, value=-0.25}, {interpMode=RCIM_Cubic, leaveTangent=0.1844, time=1, value=1.4}]}}}
     Scale Curve=0.5
    )
    Index 0 Param 4=1
    Module.Param0WriteEnabled=true
    Module.Param1WriteEnabled=true
    Module.Param2WriteEnabled=true
    Module.Param3WriteEnabled=true
   )
  ]
  rendererClasses [&NiagaraSpriteRendererProperties]
  renderers [{rendererClass=&NiagaraSpriteRendererProperties}]
 }
 Corona {
  emitterUpdateScript [
   &EmitterState()
   &SpawnBurst_Instantaneous(Use Spawn Probability=false, UseLoopCountLimit=false)
  ]
  particleSpawnScript [
   /Niagara/Modules/Spawn/Initialization/V2/InitializeParticle(
    Color=<rgba 0.15 0.5 0.5 1>
    Lifetime=2.5
    Position Mode=::NewEnumerator1
    Sprite Size Mode=::NewEnumerator3
    Uniform Sprite Size=40
    UsePositionOffset=false
    Write Material Random=true
   )
   &DynamicMaterialParameters(
    Index 0 Param 1=1
    Index 0 Param 2=0.7
    Module.Param0WriteEnabled=true
    Module.Param1WriteEnabled=true
    Module.Param2WriteEnabled=true
    Module.Param3WriteEnabled=true
   )
  ]
  particleUpdateScript [
   &ParticleState()
   &SkeletalMeshLocation(MaskSpawnGroup=false, Mesh Sampling Type=::NewEnumerator5, Skeletal Mesh=$User.SkeletalMesh, Socket Sampling Mode=::NewEnumerator5)
   /SetVariables(Particles.SpriteFacingFlat:Vector3f=<0 1 0.5>) {isSetParameters=true}
   /SetVariables(Particles.SpriteSize001:Vector2f=&Multiply_Vector2D_ByFloat(Float=0.5, Vector2D=$Particles.SpriteSize)) {isSetParameters=true}
   /SetVariables(
    Particles.ColorLight:LinearColor=&MultiplyLinearColor(Linear Color=$Particles.Color, Scale Factor=<1 1 1 1>)
    Particles.Position001:NiagaraPosition=&ConvertVectorToPosition(
     Input Position=&Multiply_Vector(
      A=&ConvertPositionToVector(Input Position=$Particles.Position, Convert To Absolute Large World Vector=false)
      B=<1 1.2 1>
     )
     Vector Is In Absolute Large World Space:NiagaraBool=false
    )
   ) {isSetParameters=true}
   &CameraOffset(Camera Offset Amount=15)
   &DynamicMaterialParameters(
    Index 0 Param 1=&FloatFromCurve(FloatCurve=DataInterface() {propertyValues {Curve {keys [<at 0 0>,<at 1 1>,<at 2 1>,<at 2.5 0>]}}}, CurveIndex=$Emitter.Age)
    Index 0 Param 2=&FloatFromCurve(FloatCurve=DataInterface() {propertyValues {Curve {keys [
         {arriveTangent=-4.2274, interpMode=RCIM_Cubic, leaveTangent=-4.2274, tangentMode=RCTM_User, time=0, value=1}
         <at 0.5 0 RCIM_Cubic>
         {interpMode=RCIM_Cubic, tangentMode=RCTM_User, time=0.8, value=0}
         <at 1 1>
        ]}}})
    Index 0 Param 4=0.2
    Module.Param0WriteEnabled=false
    Module.Param1WriteEnabled=true
    Module.Param2WriteEnabled=true
    Module.Param3WriteEnabled=true
   ) {module=DynamicMaterialParameters001}
   &ScaleColor(
    Scale Alpha=&FloatFromCurve(FloatCurve=DataInterface() {propertyValues {Curve {keys [
         {interpMode=RCIM_Cubic, tangentMode=RCTM_User, time=0, value=1}
         {arriveTangent=-0.7812, interpMode=RCIM_Cubic, leaveTangent=-0.7812, time=0.5, value=0}
         {arriveTangent=2, interpMode=RCIM_Cubic, leaveTangent=2, time=0.8, value=0.01}
         {tangentMode=RCTM_User, time=1, value=0.2}
        ]}}})
    ScaleA=true
    ScaleRGB=true
   )
  ]
  rendererClasses [&NiagaraSpriteRendererProperties, &NiagaraLightRendererProperties]
  renderers [
   {rendererClass=&NiagaraSpriteRendererProperties}
   {rendererClass=&NiagaraSpriteRendererProperties, rendererIndex=1}
   {rendererClass=&NiagaraLightRendererProperties, rendererIndex=2}
  ]
 }
 Cube_Trails_Small {
  bEnabled=false
  emitterUpdateScript [
   &EmitterState()
   &SpawnRate(SpawnRate=20, Use Spawn Probability=true)
   &SpawnRate(SpawnGroup=1, SpawnRate=20, Use Spawn Probability=true) {module=SpawnRate002}
   &SpawnRate(Spawn Probability=$User.HelmetOff, SpawnGroup=2, SpawnRate=30, Use Spawn Probability=true) {module=SpawnRate001}
  ]
  particleSpawnScript [
   /SetVariables(Particles.SpawnGroup=$Emitter.SpawnGroup) {isSetParameters=true}
   &InitializeParticle(
    Color=<rgba 2 2 2 1>
    Lifetime Max=1
    Lifetime Min=0.75
    Lifetime Mode=::NewEnumerator1
    Mesh Scale=&MakeVector(
     X=&RandomRangeFloat(Minimum=1.5, Maximum=3)
     Y=&RandomRangeFloat(Minimum=0.5, Maximum=1)
     Z=&RandomRangeFloat(Minimum=0.5, Maximum=1)
    )
    Mesh Scale Mode=::NewEnumerator4
    UsePositionOffset=false
    Write Material Random=true
   )
   &InitialMeshOrientation(Mesh Orientation Mode=::NewEnumerator2, Rotation Coordinate Space=::NewEnumerator2, Use Rotation Vector=false)
  ]
  particleUpdateScript [
   &ParticleState()
   &SkeletalMeshLocation(
    Bone / Socket Index=$Particles.SpawnGroup
    Bone Sampling Mode=::NewEnumerator4
    MaskSpawnGroup=false
    Orientation Sampling=::NewEnumerator1
    Sampled Position Offset=&SelectVectorFromArray(Direct Array Index=$Particles.SpawnGroup, Vector Selection Array=$System.PositionOffset)
    Skeletal Mesh=$User.SkeletalMesh
   )
   /SetVariables(Particles.Position=&Add_Vector(
     A=&ConvertPositionToVector(Input Position=$Particles.Position, Convert To Absolute Large World Vector=false)
     B=&Multiply_VectorByFloat(
      Vector=&Add_Vector(
       A=$System.DiveVelocityScale
       B=&RotateVectorByQuaternion(
        Vector To Rotate=&SelectVectorFromArray(Direct Array Index=$Particles.SpawnGroup, Vector Selection Array=$System.AngleOffset)
        Quaternion=$Particles.SkeletalMeshLocation.SampledRotation
       )
      )
      Float=&FloatFromCurve(
       FloatCurve=DataInterface() {propertyValues {Curve {keys [
           {arriveTangent=0.5799, interpMode=RCIM_Cubic, leaveTangent=0.5799, tangentMode=RCTM_User, time=0, value=-0.1}
           {arriveTangent=1.4467, interpMode=RCIM_Cubic, leaveTangent=1.4467, tangentMode=RCTM_User, time=1, value=1}
          ]}}}
       Scale Curve=&Multiply_Float(
        A=&MaskFloatBySpawnGroup(Masked Float=20, Passthrough Float=15, Spawn Group Mask=2)
        B=&RandomRangeFloat(Minimum=0.5, Maximum=1)
       )
      )
     )
    )) {isSetParameters=true}
   &RotateAroundPoint(
    Derive the Rotation Axis=false
    Radius:NiagaraFloat=&Multiply_Float(
     A=&Lerp_Float(
      A=&RandomRangeFloat(Minimum=1, Maximum=1.5)
      B=&RandomRangeFloat(Minimum=3, Maximum=6)
      Alpha=&FloatFromCurve(FloatCurve=DataInterface() {propertyValues {Curve {keys [
           {arriveTangent=3.8353, interpMode=RCIM_Cubic, leaveTangent=3.8353, tangentMode=RCTM_User, time=0, value=0.6}
           {interpMode=RCIM_Cubic, tangentMode=RCTM_User, time=0.25, value=0.9847}
           {interpMode=RCIM_Cubic, leaveTangent=-1.7958, time=1, value=0}
          ]}}})
     )
     B=&MaskFloatBySpawnGroup(Masked Float=1, Passthrough Float=0.7, Spawn Group Mask=2)
    )
    Rotation Axis:Vector3f=$System.DiveVelocityScale
    Rotation Phase:NiagaraFloat=&RandomRangeFloat(Minimum=0, Maximum=1)
    Rotation Rate:NiagaraFloat=&MakeCustomFloatFromBool(Bool=&RandomBool(), True Float=0.6, False Float=-0.6)
    Set Particle Position=false
   )
   /SetVariables(Particles.Rotational_Position:NiagaraPosition=&ConvertVectorToPosition(
     Input Position=&Add_Vector(
      A=&ConvertPositionToVector(Input Position=$Particles.Position, Convert To Absolute Large World Vector=false)
      B=Linked() {linkedVariable {name=Output.RotateAroundPoint.OutputPosition, type=NiagaraPosition}}
     )
     Vector Is In Absolute Large World Space:NiagaraBool=true
    )) {isSetParameters=true}
   &UpdateMeshOrientation(
    Facing Direction=&Add_Vector(
     A=$System.DiveVelocityScale
     B=&RotateVectorByQuaternion(
      Vector To Rotate=&SelectVectorFromArray(Direct Array Index=$Particles.SpawnGroup, Vector Selection Array=$System.AngleOffset)
      Quaternion=$Particles.SkeletalMeshLocation.SampledRotation
     )
    )
    Facing Interpolation=false
    Orientation Method=::NewEnumerator1
    Override Quat=false
    Side Direction:Vector3f=$Particles.RotateAroundPoint.VectorFromCenter
    Stabilization Mode:ENiagara_OrientVectorCount=::NewEnumerator2
    TransformFacing=false
    TransformSide=false
   )
   &ScaleMeshSize(Scale Factor=&Multiply_VectorByFloat(
     Vector=&Lerp_Vector(
      A=<0.5 1 1>
      B=<0.75 1 1>
      Alpha=&FloatFromCurve(FloatCurve=DataInterface() {propertyValues {Curve {keys [
           {arriveTangent=0.9666, interpMode=RCIM_Cubic, time=0.2, value=0}
           {interpMode=RCIM_Cubic, tangentMode=RCTM_User, time=0.5, value=1}
           {interpMode=RCIM_Cubic, tangentMode=RCTM_User, time=0.8, value=1}
           <at 1 0 RCIM_Cubic>
          ]}}})
     )
     Float=&FloatFromCurve(
      FloatCurve=DataInterface() {propertyValues {Curve {keys [
          {arriveTangent=3.3011, interpMode=RCIM_Cubic, leaveTangent=3.3011, tangentMode=RCTM_User, time=0, value=0.25}
          {interpMode=RCIM_Cubic, tangentMode=RCTM_User, time=0.15, value=1}
          {interpMode=RCIM_Cubic, tangentMode=RCTM_Break, time=0.5, value=1}
          {arriveTangent=-3.8115, interpMode=RCIM_Constant, leaveTangent=-3.8115, time=1, value=0.25}
         ]}}}
      Scale Curve=0.75
     )
    ))
   &DynamicMaterialParameters(
    Index 0 Param 1=&FloatFromCurve(
     FloatCurve=DataInterface() {propertyValues {Curve {keys [{interpMode=RCIM_Cubic, tangentMode=RCTM_User, time=0.6, value=0}, {interpMode=RCIM_Cubic, tangentMode=RCTM_User, time=1, value=1}]}}}
     Scale Curve=6
    )
    Module.Param0WriteEnabled=true
    Module.Param1WriteEnabled=false
    Module.Param2WriteEnabled=false
    Module.Param3WriteEnabled=false
   )
   &Color(Color=&LerpLinearColors(
     StartColor=$User.ColorA
     EndColor=$User.ColorB
     LerpFactor=&FloatFromCurve(FloatCurve=DataInterface() {propertyValues {Curve {keys [{interpMode=RCIM_Cubic, tangentMode=RCTM_User, time=0.6, value=0}, {interpMode=RCIM_Cubic, leaveTangent=5.1459, time=0.75, value=1}]}}})
    ))
   &CameraOffset(Camera Offset Amount=&MaskFloatBySpawnGroup(Masked Float=0.5, Passthrough Float=0.5, Spawn Group Mask=2))
  ]
  rendererClasses [&NiagaraMeshRendererProperties]
  renderers [{rendererClass=&NiagaraMeshRendererProperties, rendererIndex=0}]
 }
 SubUV_ConeGlowLoop {
  emitterUpdateScript [
   &EmitterState(Life Cycle Mode=::NewEnumerator1, Loop Behavior=::NewEnumerator1, Loop Duration=0.2, UseLoopDelay=false)
   &SpawnRate(SpawnRate=80, Use Spawn Probability=false)
   &SpawnBurst_Instantaneous(Use Spawn Probability=false, UseLoopCountLimit=false) {enabled=false}
   &SpawnBurst_Instantaneous(Spawn Count=11, Use Spawn Probability=false, UseLoopCountLimit=false) {enabled=false, module=SpawnBurst_Instantaneous001}
  ]
  particleSpawnScript [
   &InitializeParticle(
    Color=<rgba 1 1 1 0.4>
    Lifetime Max=1
    Lifetime Min=0.5
    Lifetime Mode=::NewEnumerator1
    Sprite Size Max=<40 30>
    Sprite Size Min=<10 10>
    Sprite Size Mode=::NewEnumerator2
    UsePositionOffset=false
    Write Material Random=true
   )
   &AddVelocity(Velocity=<1 0 0>) {enabled=false}
   &SubUVAnimation(End Frame=0)
  ]
  particleUpdateScript [
   &SolveForcesAndVelocity(Clamp Velocity=false, Limit Acceleration=false)
   &ParticleState(Kill Particles When Lifetime Has Elapsed=false)
   &SkeletalMeshLocation(MaskSpawnGroup=false, Mesh Sampling Type=::NewEnumerator5, OverrideRandomSeed=false, Skeletal Mesh=$User.SkeletalMesh)
   &SpriteFacingAndAlignment(
    Alignment Coordinate Space=::Local
    Sprite Alignment=&RotateVectorByQuaternion(Vector To Rotate=<1 0 0>, Quaternion=$Particles.SkeletalMeshLocation.SampledRotation)
    Sprite Facing=<0 1 1>
    WriteAlignment=true
    WriteFacing=false
   )
   &ScaleSpriteSize(
    Non-Uniform Curve Sprite Scale=DataInterface() {propertyValues {
      XCurve {keys [
        {arriveTangent=5.8588, interpMode=RCIM_Cubic, leaveTangent=5.8588, tangentMode=RCTM_User, time=0, value=0}
        {arriveTangent=-0.8935, interpMode=RCIM_Cubic, leaveTangent=-0.8935, tangentMode=RCTM_User, time=1, value=1.375}
        {arriveTangent=-0.6275, interpMode=RCIM_Cubic, leaveTangent=-0.6275, tangentMode=RCTM_User, time=2.5, value=0}
       ]}
      YCurve {keys [
        {arriveTangent=3.9464, interpMode=RCIM_Cubic, leaveTangent=3.9464, tangentMode=RCTM_User, time=0, value=0.1}
        {interpMode=RCIM_Cubic, tangentMode=RCTM_User, time=0.5619, value=0.8648}
        {arriveTangent=-0.9345, interpMode=RCIM_Cubic, leaveTangent=-0.9345, time=0.752, value=0.6956}
        {arriveTangent=-0.3605, interpMode=RCIM_Cubic, leaveTangent=-0.3605, time=1.3, value=0.175}
        <at 2.5176 0.059 RCIM_Cubic>
       ]}
     }}
    Scale Sprite Size Mode=::NewEnumerator3
   )
   &ScaleColor(
    Scale Alpha=&FloatFromCurve(FloatCurve=DataInterface() {propertyValues {Curve {keys [
         {arriveTangent=11.3841, interpMode=RCIM_Cubic, leaveTangent=11.3841, tangentMode=RCTM_User, time=0, value=0}
         {arriveTangent=3.3333, interpMode=RCIM_Cubic, leaveTangent=3.3333, time=0.1, value=2}
         {arriveTangent=-0.6674, interpMode=RCIM_Cubic, leaveTangent=-0.6674, time=0.3, value=1}
         {arriveTangent=-0.2894, interpMode=RCIM_Cubic, leaveTangent=-0.2894, tangentMode=RCTM_User, time=1.8187, value=0.8529}
         {arriveTangent=2.9353, interpMode=RCIM_Cubic, leaveTangent=2.9353, time=1.9187, value=2.5}
         {arriveTangent=-3.9339, interpMode=RCIM_Cubic, leaveTangent=-3.9339, time=2.0187, value=1.44}
         {arriveTangent=-6.3245, interpMode=RCIM_Cubic, leaveTangent=-6.3245, tangentMode=RCTM_User, time=2.5, value=0.2131}
        ]}}})
    Scale RGB=&MakeVectorFromLinearColorRGB(Color=&ColorFromCurve(ColorCurve=DataInterface() {propertyValues {
        AlphaCurve {keys [<at 0 1 RCIM_Cubic>,<at 1 1 RCIM_Cubic>]}
        BlueCurve {keys [<at 0.4 1 RCIM_Cubic>,<at 0.7 0.2>]}
        GreenCurve {keys [<at 0.4 0.8726 RCIM_Cubic>,<at 0.7 0.1731>]}
        RedCurve {keys [<at 0.4 0.5 RCIM_Cubic>,<at 0.7 0.1>]}
       }}))
    ScaleA=true
    ScaleRGB=true
   )
   &CameraOffset(Camera Offset Amount=4)
   &DynamicMaterialParameters(Module.Param0WriteEnabled=true, Module.Param1WriteEnabled=true, Module.Param2WriteEnabled=true, Module.Param3WriteEnabled=true)
  ]
  rendererClasses [&NiagaraSpriteRendererProperties]
  renderers [{rendererClass=&NiagaraSpriteRendererProperties}]
 }
}
summary {
 systemName=NS_Emote_Origin1_Warmup
 emitters [
  {bEnabled=false, emitterName=Cube_Trails_Small, rendererClasses [&NiagaraMeshRendererProperties]}
  {bEnabled=false, emitterName=ChromaSprites, rendererClasses [&NiagaraSpriteRendererProperties]}
  {bEnabled=false, emitterName=BlackDecals, rendererClasses [&NiagaraSpriteRendererProperties]}
  {emitterName=BlueFlares, rendererClasses [&NiagaraSpriteRendererProperties, &NiagaraMeshRendererProperties]}
  {emitterName=Corona, rendererClasses [&NiagaraSpriteRendererProperties, &NiagaraLightRendererProperties]}
  {emitterName=SubUV_ConeGlowLoop, rendererClasses [&NiagaraSpriteRendererProperties]}
  {
   bEnabled=false
   emitterName=BlueFlares001
   rendererClasses [&NiagaraSpriteRendererProperties, &NiagaraMeshRendererProperties]
  }
 ]
 userVariables [
  User.bIsFrontEnd {type=NiagaraBool}
  User.bIsFrontEndPreview {type=NiagaraBool, defaultValue=true}
  User.Color_01 {type=LinearColor, description="", defaultValue {struct=LinearColor, value=<rgba 1 1 1 1>}}
  User.Color_02 {type=LinearColor, description="", defaultValue {struct=LinearColor, value=<rgba 1 1 1 1>}}
  User.ColorA {type=LinearColor, description="", defaultValue {struct=LinearColor, value=<rgba 0.4938 0 2 1>}}
  User.ColorB {type=LinearColor, description="", defaultValue {struct=LinearColor, value=<rgba 1.25 0.3125 0 1>}}
  User.HelmetOff {type=NiagaraFloat, defaultValue=1, description=""}
  User.HelmetOff_ {type=NiagaraBool, defaultValue=true}
  User.IsSwing {type=NiagaraBool}
  User.ParaGlideLeanAlpha {type=NiagaraFloat, defaultValue=0, description=""}
  User.RibbonIDOffset {type=NiagaraInt32}
  User.SkeletalMesh {
   type=NiagaraDataInterfaceSkeletalMesh
   defaultValue {
    struct=NiagaraExt_VariableValue_DataInterface
    value {dataInterface=&self:NiagaraDataInterfaceSkeletalMesh_0}
   }
  }
 ]
}
