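# Niagara effect for the Origin Prison male character's second emote, playing a glitchy digital flash over the body.
# A flash sprite, mesh-sampled embers and bone-pinned chroma sprites layer with user-driven colors, some bursts gated on front-end preview.
# GPU Digital_Bits emitters sample the skeletal mesh through an expanding spherical shell mask, one adding offset color copies for chromatic aberration.
# Several emitters, including Beam, Blocks, decals and SmokeLoop, are disabled.
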NiagaraSystem /BRCosmetics/Characters/Player/Male/Medium/Bodies/M_MED_Origin_Prison/FX/NS_Emote_Origin_Prison2
dependencies {
 usedDataInterfaces [
  &NiagaraDataInterfaceCurve
  &NiagaraDataInterfaceColorCurve
  &NiagaraDataInterfaceVector2DCurve
  &NiagaraDataInterfaceVectorCurve
  &NiagaraDataInterfaceArrayInt32
 ]
 usedDynamicInputs []
 usedModules [&SystemState]
}
emitters {
 Beam {
  bEnabled=false
  emitterUpdateScript [
   &EmitterState(
    Enable Distance Culling=true
    Max Distance Response=::SleepAndClearParticles
    MaxDistance=7000
    Module.bUseMaxDistance=true
    Module.bUseMinDistance=false
    Reset Age On Awaken=true
    Scalability Mode=::NewEnumerator1
   )
   &SpawnBurst_Instantaneous(Spawn Count=1, Use Spawn Probability=false)
  ]
  particleSpawnScript [
   /Niagara/Modules/Spawn/Initialization/InitializeParticle(
    Color=<rgba 0.8317 1.7858 2 0.1>
    Lifetime=0.65
    Mass=&RandomRangeFloat(Minimum=0.6, Maximum=1.2)
    Module.Write Color=true
    Module.Write Lifetime=true
    Module.Write Mass=false
    Module.Write Material Random=true
    Module.Write Position=false
    Module.Write Scale=true
    Module.Write SpriteRotation=true
    Module.Write SpriteSize=true
    Sprite Rotation=&NormalizedAngleToDegrees(NormalizedAngle=&RandomRangeFloat(Minimum=0, Maximum=1))
    Sprite Size=&Vector2DFromFloat(Value=&RandomRangeFloat(Minimum=120, Maximum=150))
   )
   &SphereLocation(
    MaskSpawnGroup=false
    Offset=<0 0 -10>
    Override Seed=false
    Sphere Radius=1
    Surface Only=false
    UseNonUniformScale=false
    UseOffset=true
    UseSphereOrientation=false
   )
   &DynamicMaterialParameters(
    Index 0 Param 1=220
    Index 0 Param 3=&RandomRangeFloat(Minimum=0.4, Maximum=0.6)
    Index 0 Param 4=1
    Module.Param0WriteEnabled=false
    Module.Param1WriteEnabled=false
    Module.Param2WriteEnabled=false
    Module.Param3WriteEnabled=false
   )
   &InitialMeshOrientation(Orientation Vector=<0 0 1>, Use Orientation Vector=true, Use Rotation Vector=false) {module=InitialMeshOrientation001}
  ]
  particleUpdateScript [
   &ParticleState(Kill Particles When Lifetime Has Elapsed=false, Let Infinitely Lived Particles Die When Emitter Deactivates=true)
   &ScaleColor(
    Scale Alpha=&FloatFromCurve(FloatCurve=DataInterface() {propertyValues {Curve {keys [<at 0 0>,<at 0.25 1>,<at 1 0>]}}})
    ScaleA=true
    ScaleRGB=true
   )
   &DynamicMaterialParameters(
    Index 0 Param 1=&FloatFromCurve(FloatCurve=DataInterface() {propertyValues {Curve {keys [
         {arriveTangent=2175.405, interpMode=RCIM_Cubic, leaveTangent=2175.4055, tangentMode=RCTM_User, time=0, value=0}
         {interpMode=RCIM_Cubic, tangentMode=RCTM_User, time=0.25, value=220}
         {interpMode=RCIM_Cubic, tangentMode=RCTM_User, time=1, value=0}
        ]}}})
    Module.Param0WriteEnabled=true
    Module.Param1WriteEnabled=false
    Module.Param2WriteEnabled=false
    Module.Param3WriteEnabled=false
   ) {module=DynamicMaterialParameters001}
   &FN_Light_Attributes(
    Color=<rgba 0.4 0.79 1 1>
    Create Color Attribute=true
    Create Exponent Attribute=true
    Create Position Attribute=true
    Create Radius Attribute=true
    Exponent=2
    Position=<0 0 100>
    Radius=150
    Scale Color=&FloatFromCurve(
     FloatCurve=DataInterface() {propertyValues {Curve {keys [
         {arriveTangent=74.216, interpMode=RCIM_Cubic, leaveTangent=74.2159, tangentMode=RCTM_User, time=0, value=0}
         {interpMode=RCIM_Cubic, tangentMode=RCTM_User, time=0.25, value=10}
         <at 1 0>
        ]}}}
     Scale Curve=&RandomRangeFloat(Minimum=0.2, Maximum=1, Evaluation Type=::NewEnumerator1)
    )
   )
  ]
  rendererClasses [&NiagaraMeshRendererProperties, &NiagaraLightRendererProperties]
  renderers [
   {rendererClass=&NiagaraMeshRendererProperties, rendererIndex=0}
   {rendererClass=&NiagaraLightRendererProperties, rendererIndex=1}
  ]
 }
 BlackDecals {
  bEnabled=false
  emitterUpdateScript [
   &EmitterState(Life Cycle Mode=::NewEnumerator1, Loop Behavior=::NewEnumerator1, Loop Delay=0.15, Loop Duration=1, UseLoopDelay=true)
   &SpawnBurst_Instantaneous(
    Spawn Count=20
    Spawn Probability=&MakeCustomFloatFromBool(Bool=$User.bIsFrontEndPreview, True Float=1, False Float=0)
    Spawn Time=0.01
    Use Spawn Probability=true
   )
  ]
  particleSpawnScript [/Niagara/Modules/Spawn/Initialization/V2/InitializeParticle(
    Color=<rgba 0 0 0 1>
    Lifetime Max=0.9
    Lifetime Min=0.7
    Lifetime Mode=::NewEnumerator1
    Sprite Size Mode=::NewEnumerator3
    Uniform Sprite Size=50
    UsePositionOffset=false
    Write Material Random=true
   ) {module=InitializeParticle001}]
  particleUpdateScript [
   &ParticleState(Kill Particles When Lifetime Has Elapsed=false)
   &Color(Color=&MultiplyLinearColor(
     Linear Color=<rgba 0 0 0 1>
     Scale Factor=&MakeFloat4FromVectorAndFloat("Vector (RGB)"=&VectorFromFloat(Value=3), "Float (Alpha)"=1)
    ))
   &ScaleColor(
    Color Value To Scale=$Particles.Color
    Scale Alpha=&FloatFromCurve(
     FloatCurve=DataInterface() {propertyValues {Curve {keys [
         {arriveTangent=7.9503, interpMode=RCIM_Cubic, time=0, value=0}
         {arriveTangent=1.6667, interpMode=RCIM_Cubic, leaveTangent=1.6667, time=0.125, value=1.25}
         {arriveTangent=-1.4286, interpMode=RCIM_Cubic, leaveTangent=-1.4286, time=0.6, value=1}
         <at 1 0>
        ]}}}
     Scale Curve=80
    )
    Scale RGB=&VectorFromFloat(Value=1)
    ScaleA=true
    ScaleRGB=true
   )
   &SkeletalMeshLocation(
    Bone / Socket Index=&ReturnParticleIDAttributes()
    Bone Sampling Mode=::NewEnumerator4
    Inherit Velocity=false
    MaskSpawnGroup=false
    Skeletal Mesh=DataInterface() {propertyValues {
      DefaultMesh=null
      ExcludeBoneName=root
      OptionalComponentClass=null
      PreviewMesh=/Game/Characters/Player/Female/Medium/Bodies/F_MED_Scholar/Meshes/F_MED_Scholar
      SkinningMode=SkinOnTheFly
      SoftSourceActor=null
      SourceMode=Default
      UvSetIndex=0
      WholeMeshLOD=-1
      bExcludeBone=false
      bReadDeformedGeometry=true
      bRequireCurrentFrameData=true
      ComponentTags []
      FilteredBones [
       pelvis
       spine_02
       upperarm_l
       lowerarm_l
       thigh_twist_01_l
       upperarm_r
       lowerarm_r
       thigh_twist_01_r
       neck_01
       thigh_l
       calf_l
       spine_05
       thigh_r
       calf_r
       foot_r
       foot_l
       head
       spine_04
       hand_r
       hand_l
      ]
      FilteredSockets []
      MeshUserParameter {parameter {name=None, typeDefHandle {registeredTypeIndex=1}, varData []}}
      SamplingRegions []
     }}
   )
   &CameraOffset(Camera Offset Amount=25)
   &DynamicMaterialParameters(
    Index 0 Param 1=&FloatFromCurve(FloatCurve=DataInterface() {propertyValues {Curve {keys [
         <at 0 -0.5 RCIM_Cubic>
         {arriveTangent=10, interpMode=RCIM_Cubic, leaveTangent=10, time=0.0692, value=0.2738}
         {arriveTangent=5.0777, interpMode=RCIM_Cubic, leaveTangent=5.0777, time=0.15, value=1}
         {interpMode=RCIM_Cubic, leaveTangent=4.2839, time=1, value=5}
        ]}}})
    Index 0 Param 3=&FloatFromCurve(
     FloatCurve=DataInterface() {propertyValues {Curve {keys [{arriveTangent=2.3217, interpMode=RCIM_Cubic, time=0, value=-0.25}, {interpMode=RCIM_Cubic, leaveTangent=0.1844, time=1, value=1.4}]}}}
     Scale Curve=0.5
    )
    Index 0 Param 4=1
    Module.Param0WriteEnabled=true
    Module.Param1WriteEnabled=true
    Module.Param2WriteEnabled=true
    Module.Param3WriteEnabled=true
   )
  ]
  rendererClasses [&NiagaraSpriteRendererProperties]
  renderers [{rendererClass=&NiagaraSpriteRendererProperties}]
 }
 Blocks {
  bEnabled=false
  emitterUpdateScript [
   &EmitterState()
   &SpawnBurst_Instantaneous(Spawn Count=20, Use Spawn Probability=false)
   &SpawnRate(
    SpawnRate=&FloatFromCurve(FloatCurve=DataInterface() {propertyValues {Curve {keys [<at 0.1 28>,<at 0.6 0>]}}}, CurveIndex=$Emitter.NormalizedLoopAge)
    Use Spawn Probability=false
   )
  ]
  particleSpawnScript [
   /SetVariables(Particles.Random=&RandomRangeFloat(Minimum=0, Maximum=1)) {isSetParameters=true, module=SetVariables_#1}
   /Niagara/Modules/Spawn/Initialization/InitializeParticle(
    Color=&LerpLinearColors(StartColor=<rgba 0 0.9143 1 1>, EndColor=<rgba 0.6833 0 1 1>, LerpFactor=&UniformAOrBFloat(A=0))
    Lifetime=&Lerp_Float(A=0.3, B=0.5, Alpha=$Particles.Random)
    Module.Write Color=true
    Module.Write Lifetime=true
    Module.Write Mass=true
    Module.Write Material Random=true
    Module.Write Position=false
    Module.Write Scale=false
    Module.Write SpriteRotation=false
    Module.Write SpriteSize=true
    Position=&SimulationPosition()
    Sprite Size=&Lerp_Vector2(A=<8 12.5>, B=<5 10>, Alpha=$Particles.Random)
   )
   &FN_SetSpawnGroup_v2(Number of Spawn Groups=17) {enabled=false}
   &DynamicMaterialParameters(
    Index 0 Param 1=&Lerp_Float(A=0.1, B=0.3, Alpha=$Particles.Random)
    Index 0 Param 2=&Lerp_Float(A=1.5, B=2.5, Alpha=$Particles.Random)
    Index 0 Param 3=&Lerp_Float(A=0.1, B=0.5, Alpha=$Particles.Random)
    Index 0 Param 4=1
    Module.Param0WriteEnabled=true
    Module.Param1WriteEnabled=true
    Module.Param2WriteEnabled=true
    Module.Param3WriteEnabled=true
   )
   &CameraOffset(Camera Offset Amount=20)
   &SphereLocation(
    MaskSpawnGroup=false
    Offset=<0 0 80>
    Offset Coordinate Space:ENiagaraCoordinateSpace=::World
    Override Seed=false
    Sphere Coordinate Space=::World
    Sphere Radius=&FloatFromCurve(FloatCurve=DataInterface() {propertyValues {Curve {keys [<at 0.1 60>,<at 0.35 5>]}}}, CurveIndex=$System.NormalizedLoopAge)
    Surface Only=false
    UseNonUniformScale=false
    UseOffset=true
    UseSphereOrientation=false
   )
   &AddVelocity(Velocity=&VectorFromCurve(
     VectorCurve=DataInterface() {propertyValues {XCurve {keys [<at 0 1>,<at 1 0>]}, YCurve {keys [<at 0 1>,<at 1 0>]}, ZCurve {keys [<at 0.1 120>,<at 0.5 300>]}}}
     CurveIndex=$System.NormalizedLoopAge
    ))
  ]
  particleUpdateScript [
   &ParticleState()
   &AccelerationForce(Acceleration=&Lerp_Vector(A=<0 0 500>, B=<0 0 50>, Alpha=$Particles.Random))
   &ScaleSpriteSize(Scale Factor=&Vector2DFromCurve(
     Vector2Curve=DataInterface() {propertyValues {
       XCurve {keys [{arriveTangent=-5.9286, time=-0.0103, value=0.856}, <at 0.1376 1.5917>, <at 1 0.25>]}
       YCurve {keys [{arriveTangent=9.8642, interpMode=RCIM_Cubic, leaveTangent=9.8642, tangentMode=RCTM_User, time=0, value=0}, <at 1 4>]}
      }}
     ScaleCurve=<1 1>
    ))
   &ScaleColor(
    Scale Alpha=&FloatFromCurve(FloatCurve=DataInterface() {propertyValues {Curve {keys [
         {arriveTangent=-0.1617, interpMode=RCIM_Cubic, leaveTangent=-0.1617, tangentMode=RCTM_User, time=0, value=0}
         {arriveTangent=-0.0664, interpMode=RCIM_Cubic, leaveTangent=-0.0664, tangentMode=RCTM_User, time=0.35, value=1}
         {arriveTangent=-0.0647, interpMode=RCIM_Cubic, leaveTangent=-0.0647, tangentMode=RCTM_User, time=0.6247, value=1}
         <at 1 0>
        ]}}})
    Scale RGB=&VectorFromFloat(Value=&FloatFromCurve(FloatCurve=DataInterface() {propertyValues {Curve {keys [<at 0 15>,<at 0.6129 0.434>,<at 1 0>]}}}))
    ScaleA=true
    ScaleRGB=true
   )
   &SolveForcesAndVelocity(Clamp Velocity=false, Limit Acceleration=false)
  ]
  rendererClasses [&NiagaraSpriteRendererProperties]
  renderers [{rendererClass=&NiagaraSpriteRendererProperties}]
 }
 BurstEmbers {
  emitterUpdateScript [
   &EmitterState(Life Cycle Mode=::NewEnumerator1, Loop Behavior=::NewEnumerator1, Loop Delay=0.15, Loop Duration=0.4, UseLoopDelay=true)
   &SpawnBurst_Instantaneous(
    Spawn Count=&RandomRangeInt(Minimum=150, Maximum=180)
    Spawn Probability=&MakeCustomFloatFromBool(Bool=$User.bIsFrontEndPreview, True Float=1, False Float=0)
    Use Spawn Probability=true
   )
  ]
  particleSpawnScript [
   &FN_SetSpawnGroup(Number of Spawn Groups=0)
   &InitializeParticle(
    Lifetime=&RandomRangeFloat(Minimum=0.5, Maximum=1, Evaluation Type=::NewEnumerator1)
    Module.Write Color=true
    Module.Write Lifetime=true
    Module.Write Mass=true
    Module.Write Material Random=true
    Module.Write Position=false
    Module.Write Scale=false
    Module.Write SpriteRotation=false
    Module.Write SpriteSize=true
    Sprite Size=&RandomRangeVector2D(Minimum=<0.5 1.25>, Maximum=<0.8 3.5>, Evaluation Type=::NewEnumerator1)
   )
   &AddVelocity(Velocity=&RandomRangeVector(Minimum=<-20 -20 -20>, Maximum=<20 20 20>, Evaluation Type=::NewEnumerator1))
   &SkeletalMeshLocation(
    Inherit Velocity=false
    MaskSpawnGroup=false
    Mesh Sampling Type=::NewEnumerator5
    Module.Override Randomness=false
    Skeletal Mesh=$User.SkeletalMesh
   )
   /Niagara/Modules/Spawn/Location/SphereLocation(MaskSpawnGroup=false, Module.Override Randomness=false, Sphere Radius=13, Surface Only=false)
   &AddVelocity(Velocity=&Multiply_VectorByFloat(
     Vector=$Output.SphereLocation.SphereVector
     Float=&RandomRangeFloat(Minimum=15, Maximum=25, Evaluation Type=::NewEnumerator1)
    )) {module=AddVelocity001}
  ]
  particleUpdateScript [
   &ParticleState()
   &ScaleSpriteSize(Scale Factor=&Vector2DFromFloat(Value=&FloatFromCurve(FloatCurve=DataInterface() {propertyValues {Curve {keys [
          {arriveTangent=11.6538, interpMode=RCIM_Cubic, time=0, value=0.0211}
          {arriveTangent=0.3289, interpMode=RCIM_Cubic, leaveTangent=0.3289, time=0.125, value=1.5}
          <at 1 0.35 RCIM_Cubic>
         ]}}})))
   &Drag(Drag=4.5, Use Linear Drag=true, Use Rotational Drag=false)
   &AccelerationForce(Acceleration=<0 0 50>)
   &SolveForcesAndVelocity(Clamp Velocity=false, Limit Acceleration=false)
   &ScaleSpriteSizeBySpeed(Max Scale Factor=<1 4>, Source Velocity:Vector3f=&Multiply_Vector(A=<0 0.01 1>, B=$Particles.Velocity))
   &Color(Color=&MultiplyLinearColor(
     Linear Color=&LerpLinearColors(
      StartColor=$User.Color_01
      EndColor=$User.Color_02
      LerpFactor=&FloatFromCurve(FloatCurve=DataInterface() {propertyValues {Curve {keys [<at 0.2 0>,<at 0.8 1>]}}})
     )
     Scale Factor=&MakeVector4_FromLinearColor(LinearColor=&RandomRangeLinearColor(Minimum=<rgba 1 1 1 1>, Maximum=<rgba 0.65 0.65 1 1>))
    )) {module=Color002}
   &ScaleColor(
    Color Value To Scale=$Particles.Color
    Scale Alpha=&FloatFromCurve(FloatCurve=DataInterface() {propertyValues {Curve {keys [<at 0 0 RCIM_Cubic>,<at 0.1 1 RCIM_Cubic>,<at 1 0>]}}})
    Scale RGB=&VectorFromFloat(Value=6)
    ScaleA=true
    ScaleRGB=true
   ) {module=ScaleColor001}
  ]
  rendererClasses [&NiagaraSpriteRendererProperties]
  renderers [{rendererClass=&NiagaraSpriteRendererProperties}]
 }
 ChromaSprites {
  emitterUpdateScript [
   &EmitterState(Life Cycle Mode=::NewEnumerator1, Loop Behavior=::NewEnumerator1, Loop Delay=0.15, Loop Duration=1, UseLoopDelay=true)
   &SpawnBurst_Instantaneous(
    Spawn Count=20
    Spawn Probability=&MakeCustomFloatFromBool(Bool=$User.bIsFrontEndPreview, True Float=1, False Float=0)
    Spawn Time=0.01
    Use Spawn Probability=true
   )
  ]
  particleSpawnScript [&InitializeParticle(
    Color=&MakeLinearColorFromVectorAndFloat("Vector (RGB)"=&Multiply_VectorByFloat(Vector=<0 0.2 0.43>, Float=0.025), "Float (Alpha)"=1)
    Lifetime=0.95
    Module.Write Color=true
    Module.Write Lifetime=true
    Module.Write Mass=true
    Module.Write Material Random=true
    Module.Write Position=false
    Module.Write Scale=false
    Module.Write SpriteRotation=false
    Module.Write SpriteSize=true
    Sprite Size=&Vector2DFromFloat(Value=50)
   )]
  particleUpdateScript [
   &ParticleState(Kill Particles When Lifetime Has Elapsed=false)
   &Color(Color=&MultiplyLinearColor(
     Linear Color=&MultiplyLinearColor(Linear Color=$User.Color_01, Scale Factor=<0.1 0.1 0.1 0.2>)
     Scale Factor=&MakeFloat4FromVectorAndFloat("Vector (RGB)"=&VectorFromFloat(Value=3), "Float (Alpha)"=1)
    ))
   &ScaleColor(
    Color Value To Scale=$Particles.Color
    Scale Alpha=&FloatFromCurve(FloatCurve=DataInterface() {propertyValues {Curve {keys [{arriveTangent=7.9503, interpMode=RCIM_Cubic, time=0, value=0}, <at 0.125 1.25 RCIM_Cubic>, <at 1 0>]}}})
    Scale RGB=&VectorFromFloat(Value=1)
    ScaleA=true
    ScaleRGB=true
   )
   &SkeletalMeshLocation(
    Bone / Socket Index=&ReturnParticleIDAttributes()
    Bone Sampling Mode=::NewEnumerator4
    Inherit Velocity=false
    MaskSpawnGroup=false
    Skeletal Mesh=DataInterface() {propertyValues {
      DefaultMesh=null
      ExcludeBoneName=root
      OptionalComponentClass=null
      PreviewMesh=&F_MED_Scholar
      SkinningMode=SkinOnTheFly
      SoftSourceActor=null
      SourceMode=Default
      UvSetIndex=0
      WholeMeshLOD=-1
      bExcludeBone=false
      bReadDeformedGeometry=true
      bRequireCurrentFrameData=true
      ComponentTags []
      FilteredBones [
       pelvis
       spine_02
       upperarm_l
       lowerarm_l
       thigh_twist_01_l
       upperarm_r
       lowerarm_r
       thigh_twist_01_r
       neck_01
       thigh_l
       calf_l
       spine_05
       thigh_r
       calf_r
       foot_r
       foot_l
       head
       spine_04
       hand_r
       hand_l
      ]
      FilteredSockets []
      MeshUserParameter {parameter {name=None, typeDefHandle {registeredTypeIndex=1}, varData []}}
      SamplingRegions []
     }}
   )
   &CameraOffset(Camera Offset Amount=25) {enabled=false}
   &DynamicMaterialParameters(
    Index 0 Param 1=&FloatFromCurve(FloatCurve=DataInterface() {propertyValues {Curve {keys [
         <at 0 -0.5 RCIM_Cubic>
         {arriveTangent=10, interpMode=RCIM_Cubic, leaveTangent=10, time=0.0692, value=0.2738}
         {arriveTangent=5.0777, interpMode=RCIM_Cubic, leaveTangent=5.0777, time=0.15, value=1}
         {interpMode=RCIM_Cubic, leaveTangent=4.2839, time=1, value=5}
        ]}}})
    Index 0 Param 3=&FloatFromCurve(
     FloatCurve=DataInterface() {propertyValues {Curve {keys [{arriveTangent=2.3217, interpMode=RCIM_Cubic, time=0, value=-0.25}, {interpMode=RCIM_Cubic, leaveTangent=0.1844, time=1, value=1.4}]}}}
     Scale Curve=0.5
    )
    Index 0 Param 4=1
    Module.Param0WriteEnabled=true
    Module.Param1WriteEnabled=true
    Module.Param2WriteEnabled=true
    Module.Param3WriteEnabled=true
   )
  ]
  rendererClasses [&NiagaraSpriteRendererProperties]
  renderers [{rendererClass=&NiagaraSpriteRendererProperties}]
 }
 ChromaSprites001 {
  emitterUpdateScript [
   &EmitterState(
    Inactive Response=::NewEnumerator1
    Life Cycle Mode=::NewEnumerator1
    Loop Behavior=::NewEnumerator1
    Loop Delay=0.15
    Loop Duration=1
    UseLoopDelay=true
   )
   &SpawnBurst_Instantaneous(
    Spawn Count=20
    Spawn Probability=&MakeCustomFloatFromBool(Bool=$User.bIsFrontEndPreview, True Float=1, False Float=0)
    Spawn Time=0.01
    Use Spawn Probability=true
   ) {enabled=false}
   &SpawnRate(SpawnRate=20, Use Spawn Probability=false)
  ]
  particleSpawnScript [
   &InitializeParticle(
    Color=&MakeLinearColorFromVectorAndFloat("Vector (RGB)"=&Multiply_VectorByFloat(Vector=<0 0.2 0.43>, Float=0.025), "Float (Alpha)"=1)
    Lifetime=0.95
    Module.Write Color=true
    Module.Write Lifetime=true
    Module.Write Mass=true
    Module.Write Material Random=true
    Module.Write Position=false
    Module.Write Scale=false
    Module.Write SpriteRotation=false
    Module.Write SpriteSize=true
    Sprite Size=&Vector2DFromFloat(Value=50)
   ) {enabled=false}
   /Niagara/Modules/Spawn/Initialization/V2/InitializeParticle(
    Color=&MakeLinearColorFromVectorAndFloat("Vector (RGB)"=&Multiply_VectorByFloat(Vector=<0 0.2 0.43>, Float=0.1), "Float (Alpha)"=1)
    UsePositionOffset=false
    Write Material Random=true
   ) {module=InitializeParticle001}
  ]
  particleUpdateScript [
   &ParticleState(Kill Particles When Lifetime Has Elapsed=false)
   &Color(Color=&MultiplyLinearColor(
     Linear Color=&MakeLinearColorFromVectorAndFloat("Vector (RGB)"=&Multiply_VectorByFloat(Vector=<0.2 0.2 0>, Float=0.11), "Float (Alpha)"=1)
     Scale Factor=&MakeFloat4FromVectorAndFloat("Vector (RGB)"=&VectorFromFloat(Value=3), "Float (Alpha)"=1)
    ))
   &ScaleColor(
    Color Value To Scale=$Particles.Color
    Scale Alpha=&FloatFromCurve(FloatCurve=DataInterface() {propertyValues {Curve {keys [{arriveTangent=7.9503, interpMode=RCIM_Cubic, time=0, value=0}, <at 0.125 1.25 RCIM_Cubic>, <at 1 0>]}}})
    Scale RGB=&VectorFromFloat(Value=1)
    ScaleA=true
    ScaleRGB=true
   )
   &SkeletalMeshLocation(
    Bone / Socket Index=&ReturnParticleIDAttributes()
    Bone Sampling Mode=::NewEnumerator4
    Inherit Velocity=false
    MaskSpawnGroup=false
    Skeletal Mesh=DataInterface() {propertyValues {
      DefaultMesh=null
      ExcludeBoneName=root
      OptionalComponentClass=null
      PreviewMesh=&F_MED_Scholar
      SkinningMode=SkinOnTheFly
      SoftSourceActor=null
      SourceMode=Default
      UvSetIndex=0
      WholeMeshLOD=-1
      bExcludeBone=false
      bReadDeformedGeometry=true
      bRequireCurrentFrameData=true
      ComponentTags []
      FilteredBones [
       pelvis
       spine_02
       upperarm_l
       lowerarm_l
       thigh_twist_01_l
       upperarm_r
       lowerarm_r
       thigh_twist_01_r
       neck_01
       thigh_l
       calf_l
       spine_05
       thigh_r
       calf_r
       foot_r
       foot_l
       head
       spine_04
       hand_r
       hand_l
      ]
      FilteredSockets []
      MeshUserParameter {parameter {name=None, typeDefHandle {registeredTypeIndex=1}, varData []}}
      SamplingRegions []
     }}
   )
   &CameraOffset(Camera Offset Amount=25) {enabled=false}
   &DynamicMaterialParameters(
    Index 0 Param 1=&FloatFromCurve(FloatCurve=DataInterface() {propertyValues {Curve {keys [
         <at 0 -0.5 RCIM_Cubic>
         {arriveTangent=10, interpMode=RCIM_Cubic, leaveTangent=10, time=0.0692, value=0.2738}
         {arriveTangent=5.0777, interpMode=RCIM_Cubic, leaveTangent=5.0777, time=0.15, value=1}
         {interpMode=RCIM_Cubic, leaveTangent=4.2839, time=1, value=5}
        ]}}})
    Index 0 Param 3=&FloatFromCurve(
     FloatCurve=DataInterface() {propertyValues {Curve {keys [{arriveTangent=2.3217, interpMode=RCIM_Cubic, time=0, value=-0.25}, {interpMode=RCIM_Cubic, leaveTangent=0.1844, time=1, value=1.4}]}}}
     Scale Curve=0.5
    )
    Index 0 Param 4=1
    Module.Param0WriteEnabled=true
    Module.Param1WriteEnabled=true
    Module.Param2WriteEnabled=true
    Module.Param3WriteEnabled=true
   )
  ]
  rendererClasses [&NiagaraSpriteRendererProperties]
  renderers [{rendererClass=&NiagaraSpriteRendererProperties}]
 }
 Digital_Bits {
  simTarget=GPUComputeSim
  emitterUpdateScript [
   &EmitterState(Life Cycle Mode=::NewEnumerator1, Loop Behavior=::NewEnumerator1, Loop Delay=0.5, Loop Duration=0.5, UseLoopDelay=true)
   &SpawnRate(
    SpawnRate=&FloatFromCurve(
     FloatCurve=DataInterface() {propertyValues {Curve {keys [<at 0 1 RCIM_Cubic>, {arriveTangent=-2.7657, interpMode=RCIM_Cubic, leaveTangent=-2.7657, tangentMode=RCTM_User, time=1, value=0}]}}}
     CurveIndex=$Emitter.NormalizedLoopAge
     Scale Curve=1000
    )
    Use Spawn Probability=false
   )
  ]
  particleSpawnScript [
   /SetVariables(Particles.ChromaOffset=1) {isSetParameters=true}
   /Niagara/Modules/Spawn/Initialization/InitializeParticle(
    Color=<rgba 1 1 20 1>
    Lifetime=&UniformRangedFloat(Minimum=0.1, Maximum=0.1)
    Module.Write Color=true
    Module.Write Lifetime=true
    Module.Write Mass=true
    Module.Write Material Random=false
    Module.Write Position=false
    Module.Write Scale=false
    Module.Write SpriteRotation=false
    Module.Write SpriteSize=true
    Position=&SimulationPosition()
    Sprite Size=&RandomRangeVector2D(Minimum=<1 1>, Maximum=<2 4>)
   )
   &SkeletalMeshLocation(
    Inherit Velocity=false
    MaskSpawnGroup=false
    Mesh Sampling Type=::NewEnumerator7
    Module.Override Randomness=false
    OffsetAlongNormal=false
    Skeletal Mesh=$User.SkeletalMesh
   )
   /SetVariables(
    Particles.NormalVelocity:Vector3f=$Output.SkeletalMeshLocation.SampledNormal
    Particles.TextureSampleA:LinearColor=$Output.SampleTexture.SampledColor
    Particles.TextureSampleB:LinearColor=$Output.SampleTexture001.SampledColor
   ) {isSetParameters=true, module=SetVariables_#2}
   &CameraOffset(Camera Offset Amount=-10)
   &AddVelocity(Rotation Coordinate Space=::World, Velocity=<0 0 -20>) {module=AddVelocity001}
  ]
  particleUpdateScript [
   &ParticleState()
   &self:SphereMaskEdge(
    Cull Particles outside of max distance=true
    Max Distance:NiagaraFloat=&FloatFromCurve(
     FloatCurve=DataInterface() {propertyValues {Curve {keys [
         {arriveTangent=1.1063, interpMode=RCIM_Cubic, leaveTangent=1.1063, tangentMode=RCTM_User, time=0.001, value=0}
         {arriveTangent=1.1925, interpMode=RCIM_Cubic, leaveTangent=1.1925, tangentMode=RCTM_User, time=0.9, value=1}
        ]}}}
     CurveIndex=$Emitter.NormalizedLoopAge
     Scale Curve=128
    )
    Min Distance:NiagaraFloat=&FloatFromCurve(
     FloatCurve=DataInterface() {propertyValues {Curve {keys [
         {arriveTangent=0.7543, interpMode=RCIM_Cubic, leaveTangent=0.7543, tangentMode=RCTM_User, time=0.004, value=0}
         {arriveTangent=1.004, interpMode=RCIM_Cubic, leaveTangent=1.004, time=0.6, value=0.451}
         {arriveTangent=1.4135, interpMode=RCIM_Cubic, leaveTangent=1.4135, tangentMode=RCTM_User, time=1, value=1}
        ]}}}
     CurveIndex=$Emitter.NormalizedLoopAge
     Scale Curve=128
    )
    SphereCenter:Vector3f=<0 8.5591 128>
   ) {module=SphereMaskEdge}
   &ScaleColor(
    Scale Alpha=&Multiply_Float(
     A=&MakeFloatFromLinearColor(LinearColor=$Particles.TextureSampleA, Channel:ENiagaraLinearColor_Channels=::NewEnumerator0)
     B=&MakeFloatFromLinearColor(LinearColor=$Particles.TextureSampleB, Channel:ENiagaraLinearColor_Channels=::NewEnumerator0)
    )
    Scale RGB=&VectorFromFloat(Value=1)
    ScaleA=true
    ScaleRGB=true
   ) {enabled=false}
   &AddVelocity(Velocity=&Multiply_Vector(A=$Particles.NormalVelocity, B=&VectorFromFloat(Value=-1)))
   &SolveForcesAndVelocity(Clamp Velocity=false, Limit Acceleration=false)
   /SetVariables() {enabled=false, isSetParameters=true}
   /SetVariables(Particles.ChromaOffset:NiagaraFloat=&FloatFromCurve(
     FloatCurve=DataInterface() {propertyValues {Curve {keys [{interpMode=RCIM_Cubic, tangentMode=RCTM_User, time=0, value=1}, {interpMode=RCIM_Cubic, tangentMode=RCTM_User, time=1, value=0}]}}}
     Scale Curve=0.5
    )) {isSetParameters=true}
   /SetVariables(
    Particles.Color001:LinearColor=<rgba 0.1 0 0 0.5>
    Particles.Position001:NiagaraPosition=&ConvertVectorToPosition(
     Input Position=&Add_Vector(
      A=&ConvertPositionToVector(Input Position=$Particles.Position, Convert To Absolute Large World Vector=false)
      B=&Multiply_VectorByFloat(Vector=<2 0 1>, Float=$Particles.ChromaOffset)
     )
     Vector Is In Absolute Large World Space:NiagaraBool=false
    )
   ) {isSetParameters=true}
   /SetVariables(
    Particles.Color002=<rgba 0 0.0077 0.1 0.5>
    Particles.Position002:NiagaraPosition=&ConvertVectorToPosition(
     Input Position=&Add_Vector(
      A=&ConvertPositionToVector(Input Position=$Particles.Position, Convert To Absolute Large World Vector=false)
      B=&Multiply_VectorByFloat(Vector=<-2 -1 -2>, Float=$Particles.ChromaOffset)
     )
     Vector Is In Absolute Large World Space:NiagaraBool=false
    )
   ) {isSetParameters=true}
   /SetVariables(
    Particles.Color003:LinearColor=<rgba 0.0029 0.1 0 0.5>
    Particles.Position003:NiagaraPosition=&ConvertVectorToPosition(
     Input Position=&Add_Vector(
      A=&ConvertPositionToVector(Input Position=$Particles.Position, Convert To Absolute Large World Vector=false)
      B=&Multiply_VectorByFloat(Vector=<1 2 -2>, Float=$Particles.ChromaOffset)
     )
     Vector Is In Absolute Large World Space:NiagaraBool=false
    )
   ) {isSetParameters=true}
   &ScaleColor(Scale RGB=<5 5 5>, ScaleA=true, ScaleRGB=true) {enabled=false, module=ScaleColor001}
   &ScaleSpriteSize(Scale Factor=&Vector2DFromCurve(
     Vector2Curve=DataInterface() {propertyValues {
       XCurve {keys [
         {interpMode=RCIM_Cubic, tangentMode=RCTM_User, time=0, value=1}
         {arriveTangent=-6.968, interpMode=RCIM_Cubic, leaveTangent=-6.968, tangentMode=RCTM_User, time=0.5, value=2}
         {tangentMode=RCTM_User, time=1, value=0}
        ]}
       YCurve {keys [
         {interpMode=RCIM_Cubic, tangentMode=RCTM_User, time=0, value=0}
         {arriveTangent=13.8249, interpMode=RCIM_Cubic, leaveTangent=13.8249, tangentMode=RCTM_User, time=0.6, value=2}
         {arriveTangent=8.3745, interpMode=RCIM_Cubic, leaveTangent=8.3745, tangentMode=RCTM_User, time=1, value=8}
        ]}
      }}
     ScaleCurve=<2 8>
    ))
   /SetVariables(Particles.SpriteSize001:Vector2f=&Multiply_Vector2D_ByFloat(Float=1, Vector2D=$Particles.SpriteSize)) {isSetParameters=true}
  ]
  rendererClasses [&NiagaraSpriteRendererProperties]
  renderers [
   {rendererClass=&NiagaraSpriteRendererProperties}
   {rendererClass=&NiagaraSpriteRendererProperties, rendererIndex=1}
   {rendererClass=&NiagaraSpriteRendererProperties, rendererIndex=2}
   {rendererClass=&NiagaraSpriteRendererProperties, rendererIndex=3}
  ]
 }
 Digital_Bits001 {
  simTarget=GPUComputeSim
  emitterUpdateScript [
   &EmitterState(Life Cycle Mode=::NewEnumerator1, Loop Behavior=::NewEnumerator1, Loop Duration=0.5, UseLoopDelay=false)
   &SpawnRate(
    SpawnRate=&FloatFromCurve(
     FloatCurve=DataInterface() {propertyValues {Curve {keys [<at 0 1 RCIM_Cubic>, {arriveTangent=-2.7657, interpMode=RCIM_Cubic, leaveTangent=-2.7657, tangentMode=RCTM_User, time=1, value=0}]}}}
     CurveIndex=$Emitter.NormalizedLoopAge
     Scale Curve=1000
    )
    Use Spawn Probability=false
   )
  ]
  particleSpawnScript [
   &InitializeParticle(
    Color=&MakeLinearColorFromVectorAndFloat(
     "Vector (RGB)"=&MakeVectorFromLinearColorRGB(Color=<rgba 1 1 1 1>)
     "Float (Alpha)"=&UniformRangedFloat(Minimum=0.25, Maximum=1)
    )
    Lifetime=&UniformRangedFloat(Minimum=0.05, Maximum=0.1)
    Module.Write Color=true
    Module.Write Lifetime=true
    Module.Write Mass=true
    Module.Write Material Random=false
    Module.Write Position=false
    Module.Write Scale=false
    Module.Write SpriteRotation=false
    Module.Write SpriteSize=true
    Position=&SimulationPosition()
    Sprite Size=<1 1>
   )
   &SkeletalMeshLocation(
    Inherit Velocity=false
    MaskSpawnGroup=false
    Mesh Sampling Type=::NewEnumerator7
    Module.Override Randomness=false
    OffsetAlongNormal=false
    Skeletal Mesh=$User.SkeletalMesh
   )
   /SetVariables(
    Particles.NormalVelocity:Vector3f=$Output.SkeletalMeshLocation.SampledNormal
    Particles.TextureSampleA:LinearColor=$Output.SampleTexture.SampledColor
    Particles.TextureSampleB:LinearColor=$Output.SampleTexture001.SampledColor
   ) {isSetParameters=true, module=SetVariables_#2}
   &CameraOffset(Camera Offset Amount=50)
  ]
  particleUpdateScript [
   &ParticleState()
   &self:SphereMaskEdge(
    Cull Particles outside of max distance=true
    Max Distance:NiagaraFloat=&FloatFromCurve(
     FloatCurve=DataInterface() {propertyValues {Curve {keys [
         {arriveTangent=1.1063, interpMode=RCIM_Cubic, leaveTangent=1.1063, tangentMode=RCTM_User, time=0.001, value=0}
         {arriveTangent=1.1925, interpMode=RCIM_Cubic, leaveTangent=1.1925, tangentMode=RCTM_User, time=0.9, value=1}
        ]}}}
     CurveIndex=$Emitter.NormalizedLoopAge
     Scale Curve=128
    )
    Min Distance:NiagaraFloat=&FloatFromCurve(
     FloatCurve=DataInterface() {propertyValues {Curve {keys [
         {arriveTangent=0.7543, interpMode=RCIM_Cubic, leaveTangent=0.7543, tangentMode=RCTM_User, time=0.004, value=0}
         {arriveTangent=1.004, interpMode=RCIM_Cubic, leaveTangent=1.004, time=0.6, value=0.451}
         {arriveTangent=1.4135, interpMode=RCIM_Cubic, leaveTangent=1.4135, tangentMode=RCTM_User, time=1, value=1}
        ]}}}
     CurveIndex=$Emitter.NormalizedLoopAge
     Scale Curve=128
    )
    SphereCenter:Vector3f=<0 8.5591 128>
   ) {module=SphereMaskEdge}
   &ScaleColor(
    Scale Alpha=&Multiply_Float(
     A=&MakeFloatFromLinearColor(LinearColor=$Particles.TextureSampleA, Channel:ENiagaraLinearColor_Channels=::NewEnumerator0)
     B=&MakeFloatFromLinearColor(LinearColor=$Particles.TextureSampleB, Channel:ENiagaraLinearColor_Channels=::NewEnumerator0)
    )
    ScaleA=true
    ScaleRGB=true
   )
   &AddVelocity(Velocity=&Multiply_Vector(A=$Particles.NormalVelocity, B=&VectorFromFloat(Value=-1)))
   &SolveForcesAndVelocity(Clamp Velocity=false, Limit Acceleration=false)
   &ScaleColor(ScaleA=true, ScaleRGB=true) {module=ScaleColor001}
   &ScaleSpriteSize(Scale Factor=&Vector2DFromFloat(Value=&FloatFromCurve(FloatCurve=DataInterface() {propertyValues {Curve {keys [<at 0.5 1>,<at 1 0>]}}}, Scale Curve=3)))
  ]
  rendererClasses [&NiagaraSpriteRendererProperties]
  renderers [{rendererClass=&NiagaraSpriteRendererProperties}]
 }
 FLASH {
  emitterUpdateScript [
   &EmitterState()
   &SpawnBurst_Instantaneous(Spawn Count=1, Use Spawn Probability=false)
  ]
  particleSpawnScript [
   &CameraOffset(Camera Offset Amount=20)
   /Niagara/Modules/Spawn/Initialization/V2/InitializeParticle(
    Color=<rgba 0.1475 0.502 1 1>
    Position Offset=<0 0 120>
    Position Offset Coordinate Space=::World
    Sprite Size Mode=::NewEnumerator3
    Uniform Sprite Size=800
    UsePositionOffset=true
    Write Material Random=true
   ) {module=InitializeParticle001}
  ]
  particleUpdateScript [
   &ParticleState()
   &ScaleColor(
    Scale Alpha=&FloatFromCurve(FloatCurve=DataInterface() {propertyValues {Curve {keys [{arriveTangent=-2.9556, interpMode=RCIM_Cubic, leaveTangent=-2.9556, tangentMode=RCTM_User, time=0, value=1}, <at 1 0>]}}})
    ScaleA=true
    ScaleRGB=true
   )
   &ScaleSpriteSize(Scale Factor=&Vector2DFromFloat(Value=&FloatFromCurve(FloatCurve=DataInterface() {propertyValues {Curve {keys [
          {arriveTangent=10.3957, interpMode=RCIM_Cubic, leaveTangent=10.3957, tangentMode=RCTM_User, time=0, value=0}
          <at 0.215 1 RCIM_Cubic>
          <at 1 0>
         ]}}})))
   &SolveForcesAndVelocity(Clamp Velocity=false, Limit Acceleration=false)
  ]
  rendererClasses [&NiagaraSpriteRendererProperties]
  renderers [{rendererClass=&NiagaraSpriteRendererProperties}]
 }
 ShieldDecals {
  bEnabled=false
  emitterUpdateScript [
   &EmitterState(Life Cycle Mode=::NewEnumerator1, Loop Behavior=::NewEnumerator1, Loop Delay=0.15, Loop Duration=1, UseLoopDelay=true)
   &SpawnBurst_Instantaneous(
    Spawn Count=20
    Spawn Probability=&MakeCustomFloatFromBool(Bool=$User.bIsFrontEndPreview, True Float=1, False Float=0)
    Spawn Time=0.01
    Use Spawn Probability=true
   )
  ]
  particleSpawnScript [/Niagara/Modules/Spawn/Initialization/InitializeParticle(
    Lifetime=0.95
    Module.Write Color=true
    Module.Write Lifetime=true
    Module.Write Mass=true
    Module.Write Material Random=true
    Module.Write Position=false
    Module.Write Scale=false
    Module.Write SpriteRotation=false
    Module.Write SpriteSize=true
    Sprite Size=&Vector2DFromFloat(Value=50)
   )]
  particleUpdateScript [
   &ParticleState()
   &Color(Color=&MultiplyLinearColor(
     Linear Color=$User.Color_01
     Scale Factor=&MakeFloat4FromVectorAndFloat("Vector (RGB)"=&VectorFromFloat(Value=3), "Float (Alpha)"=1)
    ))
   &ScaleColor(
    Color Value To Scale=$Particles.Color
    Scale Alpha=&FloatFromCurve(FloatCurve=DataInterface() {propertyValues {Curve {keys [{arriveTangent=7.9503, interpMode=RCIM_Cubic, time=0, value=0}, <at 0.125 1.25 RCIM_Cubic>, <at 1 0>]}}})
    Scale RGB=&VectorFromFloat(Value=0.4)
    ScaleA=true
    ScaleRGB=true
   )
   &SkeletalMeshLocation(
    Bone / Socket Index=&ReturnParticleIDAttributes()
    Bone Sampling Mode=::NewEnumerator4
    Inherit Velocity=false
    MaskSpawnGroup=false
    Skeletal Mesh=DataInterface() {propertyValues {
      DefaultMesh=null
      ExcludeBoneName=root
      OptionalComponentClass=null
      PreviewMesh=&F_MED_Scholar
      SkinningMode=SkinOnTheFly
      SoftSourceActor=null
      SourceMode=Default
      UvSetIndex=0
      WholeMeshLOD=-1
      bExcludeBone=false
      bReadDeformedGeometry=true
      bRequireCurrentFrameData=true
      ComponentTags []
      FilteredBones [
       pelvis
       spine_02
       upperarm_l
       lowerarm_l
       thigh_twist_01_l
       upperarm_r
       lowerarm_r
       thigh_twist_01_r
       neck_01
       thigh_l
       calf_l
       spine_05
       thigh_r
       calf_r
       foot_r
       foot_l
       head
       spine_04
       hand_r
       hand_l
      ]
      FilteredSockets []
      MeshUserParameter {parameter {name=None, typeDefHandle {registeredTypeIndex=1}, varData []}}
      SamplingRegions []
     }}
   )
   &CameraOffset(Camera Offset Amount=25)
   &DynamicMaterialParameters(
    Index 0 Param 1=&FloatFromCurve(FloatCurve=DataInterface() {propertyValues {Curve {keys [
         <at 0 -0.5 RCIM_Cubic>
         {arriveTangent=10, interpMode=RCIM_Cubic, leaveTangent=10, time=0.0692, value=0.2738}
         {arriveTangent=5.0777, interpMode=RCIM_Cubic, leaveTangent=5.0777, time=0.15, value=1}
         {interpMode=RCIM_Cubic, leaveTangent=4.2839, time=1, value=5}
        ]}}})
    Index 0 Param 3=&FloatFromCurve(
     FloatCurve=DataInterface() {propertyValues {Curve {keys [{arriveTangent=2.3217, interpMode=RCIM_Cubic, time=0, value=-0.25}, {interpMode=RCIM_Cubic, leaveTangent=0.1844, time=1, value=1.4}]}}}
     Scale Curve=0.5
    )
    Index 0 Param 4=1
    Module.Param0WriteEnabled=true
    Module.Param1WriteEnabled=true
    Module.Param2WriteEnabled=true
    Module.Param3WriteEnabled=true
   )
  ]
  rendererClasses [&NiagaraSpriteRendererProperties]
  renderers [{rendererClass=&NiagaraSpriteRendererProperties}]
 }
 SmokeLoop {
  bEnabled=false
  emitterUpdateScript [
   &EmitterState()
   &SpawnBurst_Instantaneous(
    Spawn Count=&RandomRangeInt(Minimum=60, Maximum=65)
    Spawn Probability=&MakeCustomFloatFromBool(Bool=$User.bIsFrontEndPreview, True Float=1, False Float=0)
    Spawn Time=0.05
    Use Spawn Probability=true
   )
  ]
  particleSpawnScript [
   &FN_SetSpawnGroup(Number of Spawn Groups=16)
   /SetVariables(Particles.SpriteRotationRate=&RandomRangeFloat(Minimum=-0.8, Maximum=0.8, Evaluation Type=::NewEnumerator1)) {isSetParameters=true}
   &InitializeParticle(
    Lifetime=&RandomRangeFloat(Minimum=0.6, Maximum=0.8, Evaluation Type=::NewEnumerator1)
    Module.Write Color=true
    Module.Write Lifetime=true
    Module.Write Mass=true
    Module.Write Material Random=true
    Module.Write Position=false
    Module.Write Scale=false
    Module.Write SpriteRotation=true
    Module.Write SpriteSize=true
    Sprite Rotation=&NormalizedAngleToDegrees(NormalizedAngle=&RandomRangeFloat(Minimum=0, Maximum=1, Evaluation Type=::NewEnumerator1))
    Sprite Size=&Vector2DFromFloat(Value=&RandomRangeFloat(Minimum=43, Maximum=47, Evaluation Type=::NewEnumerator1))
   )
   &Color(Color=$User.Color_01)
   &CameraOffset(Camera Offset Amount=&RandomRangeFloat(Minimum=35, Maximum=45))
   &SkeletalMeshLocation(
    Bone / Socket Index=$Particles.SpawnGroup
    Bone Sampling Mode=::NewEnumerator4
    Inherit Velocity=false
    "Inherit Velocity (Scale)"=0.25
    MaskSpawnGroup=false
    Skeletal Mesh=$User.SkeletalMesh
   )
   &SphereLocation(MaskSpawnGroup=false, Module.Override Randomness=false, Sphere Radius=16, Surface Only=false)
   &AddVelocity(Velocity=&Multiply_VectorByFloat(Vector=$Output.SphereLocation.SphereVector, Float=8))
   &AddVelocity(Velocity=&RandomRangeVector(Minimum=<0 0 15>, Maximum=<0 0 25>, Evaluation Type=::NewEnumerator1)) {module=AddVelocity001}
  ]
  particleUpdateScript [
   &ParticleState()
   &ScaleColor(
    Scale Alpha=&FloatFromCurve(
     FloatCurve=DataInterface() {propertyValues {Curve {keys [
         {arriveTangent=3.4222, interpMode=RCIM_Cubic, time=0, value=0}
         <at 0.5 4 RCIM_Cubic>
         {interpMode=RCIM_Cubic, leaveTangent=-2.8405, time=1, value=0}
        ]}}}
     Scale Curve=0.15
    )
    Scale RGB=&VectorFromFloat(Value=&FloatFromCurve(FloatCurve=DataInterface() {propertyValues {Curve {keys [{arriveTangent=-2.1797, interpMode=RCIM_Cubic, time=0, value=1}, <at 1 1>]}}}))
    ScaleA=true
    ScaleRGB=true
   )
   &SpriteRotationRate(Rotation Rate=&NormalizedAngleToDegrees(NormalizedAngle=$Particles.SpriteRotationRate))
   &ScaleSpriteSize(Scale Factor=&Vector2DFromFloat(Value=&FloatFromCurve(FloatCurve=DataInterface() {propertyValues {Curve {keys [{arriveTangent=1.9916, interpMode=RCIM_Cubic, time=0, value=0.5}, <at 1 1 RCIM_Cubic>]}}})))
   &AccelerationForce(Acceleration=<0 0 40>)
   &Drag(
    Drag=&FloatFromCurve(FloatCurve=DataInterface() {propertyValues {Curve {keys [{interpMode=RCIM_Cubic, tangentMode=RCTM_User, time=0, value=3.5}, {interpMode=RCIM_Cubic, tangentMode=RCTM_User, time=1, value=8}]}}})
    Use Linear Drag=true
    Use Rotational Drag=false
   )
   &SolveForcesAndVelocity(Clamp Velocity=false, Limit Acceleration=false)
   &DynamicMaterialParameters(
    Index 0 Param 1=&FloatFromCurve(FloatCurve=DataInterface() {propertyValues {Curve {keys [<at 0 1>,<at 1 0>]}}}, Scale Curve=0.2)
    Module.Param0WriteEnabled=true
    Module.Param1WriteEnabled=false
    Module.Param2WriteEnabled=false
    Module.Param3WriteEnabled=false
   )
  ]
  rendererClasses [&NiagaraSpriteRendererProperties]
  renderers [{rendererClass=&NiagaraSpriteRendererProperties}]
 }
 SmokeStreams_Up {
  bEnabled=false
  emitterUpdateScript [
   &EmitterState(Life Cycle Mode=::NewEnumerator1, Loop Behavior=::NewEnumerator1, Loop Duration Mode=::NewEnumerator1, UseLoopDelay=false)
   &SpawnRate(SpawnRate=50, Use Spawn Probability=false) {enabled=false}
   &FN_SpawnRibbon_v2(SpawnRate=200)
  ]
  particleSpawnScript [
   /Niagara/Modules/Spawn/Initialization/V2/InitializeParticle(Color Mode=::NewEnumerator0, Lifetime=0.1, UsePositionOffset=false, Write Material Random=true) {enabled=false}
   &InitializeRibbon(Color=<rgba 0.1 0.1 0.1 1>, Lifetime=2, Ribbon Width=8, UsePositionOffset=false, Write Material Random=true)
   &AddVelocity()
   &SkeletalMeshLocation(
    Bone / Socket Index=5
    Bone Sampling Mode=::NewEnumerator5
    MaskSpawnGroup=false
    Sampled Position Offset=<0 0 -10>
    Sampled Position Offset Coordinate Space:ENiagaraSkelMeshTransforms=::NewEnumerator2
    Skeletal Mesh=$User.SkeletalMesh
   )
   &ApplyInitialForces()
  ]
  particleUpdateScript [
   &ParticleState()
   /SetVariables(
    Particles.Position001:NiagaraPosition=&ConvertVectorToPosition(
     Input Position=&Add_Vector(
      A=&ConvertPositionToVector(Input Position=$Particles.Position, Convert To Absolute Large World Vector=false)
      B=<10 0 10>
     )
     Vector Is In Absolute Large World Space:NiagaraBool=false
    )
    Particles.Position002:NiagaraPosition=&ConvertVectorToPosition(
     Input Position=&Add_Vector(
      A=&ConvertPositionToVector(Input Position=$Particles.Position, Convert To Absolute Large World Vector=false)
      B=<-15 0 5>
     )
     Vector Is In Absolute Large World Space:NiagaraBool=false
    )
    Particles.RibbonWidth001:NiagaraFloat=&Multiply_Float(A=$Particles.RibbonWidth, B=0.7)
    Particles.RibbonWidth002:NiagaraFloat=&Multiply_Float(A=$Particles.RibbonWidth, B=0.6)
   ) {isSetParameters=true, module=SetVariables_#3}
   &SolveForcesAndVelocity(Clamp Velocity=false, Limit Acceleration=false)
   &SkeletalMeshLocation(
    Bone / Socket Index=&SelectIntFromArray(Direct Array Index=3, Int Selection Array=DataInterface() {propertyValues {IntData [0,1]}})
    MaskSpawnGroup=false
    Mesh Sampling Type=::NewEnumerator5
    Position Sampling=::NewEnumerator1
    Skeletal Mesh=$User.SkeletalMesh
    Socket Sampling Mode=::NewEnumerator5
   ) {enabled=false, module=SkeletalMeshLocation001}
   &CurlNoiseForce(Pan Noise=false) {enabled=false, module=CurlNoiseForce001}
   &DynamicMaterialParameters(
    Index 0 Param 1=1
    Index 0 Param 2=1
    Index 0 Param 3=1
    Index 0 Param 4=1
    Index 1 Param 1=1
    Index 1 Param 2=1
    Index 2 Param 1=1
    Index 2 Param 2=1
    Index 2 Param 3=&FloatFromCurve(FloatCurve=DataInterface() {propertyValues {Curve {keys [
         {arriveTangent=-2.9047, interpMode=RCIM_Cubic, leaveTangent=-2.9047, tangentMode=RCTM_User, time=0, value=1}
         {arriveTangent=-1, interpMode=RCIM_Cubic, leaveTangent=-1, time=0.5, value=-1}
         {interpMode=RCIM_Cubic, tangentMode=RCTM_User, time=1, value=0}
        ]}}})
    Index 2 Param 4=&FloatFromCurve(FloatCurve=DataInterface() {propertyValues {Curve {keys [
         {arriveTangent=-2.1237, interpMode=RCIM_Cubic, leaveTangent=-2.1237, tangentMode=RCTM_User, time=0, value=1}
         {interpMode=RCIM_Cubic, tangentMode=RCTM_User, time=1, value=1}
        ]}}})
    Module.Param0WriteEnabled=true
    Module.Param0WriteEnabled1=true
    Module.Param0WriteEnabled2=true
    Module.Param0WriteEnabled3=true
    Module.Param1WriteEnabled=true
    Module.Param1WriteEnabled1=true
    Module.Param1WriteEnabled2=true
    Module.Param1WriteEnabled3=true
    Module.Param2WriteEnabled=true
    Module.Param2WriteEnabled1=true
    Module.Param2WriteEnabled2=true
    Module.Param2WriteEnabled3=true
    Module.Param3WriteEnabled=true
    Module.Param3WriteEnabled1=true
    Module.Param3WriteEnabled2=true
    Module.Param3WriteEnabled3=true
    Write Parameter Index 1=true
    Write Parameter Index 2=true
    Write Parameter Index 3=true
   ) {module=DynamicMaterialParameters001}
   &ScaleRibbonWidth(Ribbon Width Scale=&Multiply_Float(
     A=&Multiply_Float(
      A=&FloatFromCurve(FloatCurve=DataInterface() {propertyValues {Curve {keys [
           {interpMode=RCIM_Cubic, tangentMode=RCTM_User, time=0.2, value=5}
           {arriveTangent=-7.3024, interpMode=RCIM_Cubic, leaveTangent=-7.3024, time=0.3907, value=4.7777}
           {arriveTangent=-2.2599, interpMode=RCIM_Cubic, leaveTangent=-2.2599, time=0.5821, value=2.2096}
           {arriveTangent=-4.0912, interpMode=RCIM_Cubic, leaveTangent=-4.0912, time=0.8163, value=3.8158}
           {arriveTangent=0.6917, interpMode=RCIM_Cubic, leaveTangent=0.6917, tangentMode=RCTM_User, time=1, value=0.5}
          ]}}})
      B=1
     )
     B=&FloatFromCurve(
      FloatCurve=DataInterface() {propertyValues {Curve {keys [
          {interpMode=RCIM_Cubic, tangentMode=RCTM_User, time=1, value=1}
          {interpMode=RCIM_Cubic, tangentMode=RCTM_User, time=2.5, value=0.5}
          <at 4 1 RCIM_Cubic>
         ]}}}
      CurveIndex=$Emitter.Age
     )
    ))
   &CurlNoiseForce(
    Noise Frequency=5
    Noise Strength=&FloatFromCurve(
     FloatCurve=DataInterface() {propertyValues {Curve {keys [{arriveTangent=-3.2172, interpMode=RCIM_Cubic, leaveTangent=-3.2172, tangentMode=RCTM_User, time=-0.0184, value=1.5941}, <at 1.0003 0.7389>]}}}
     CurveIndex=$Emitter.Age
     Scale Curve=1500
    )
    Pan Noise=true
    Pan Noise Field=<0 0 2>
    Random Seed:NiagaraInt32=$Particles.SpawnGroup
   ) {enabled=false}
   &DynamicMaterialParameters(
    Index 0 Param 1=1
    Index 0 Param 2=1
    Index 0 Param 3=1
    Index 0 Param 4=1
    Index 1 Param 1=1
    Index 1 Param 2=1
    Index 2 Param 1=0.2
    Index 2 Param 2=0.2
    Index 2 Param 3=0.2
    Index 2 Param 4=1
    Module.Param0WriteEnabled=true
    Module.Param0WriteEnabled1=true
    Module.Param0WriteEnabled2=true
    Module.Param1WriteEnabled=true
    Module.Param1WriteEnabled1=true
    Module.Param1WriteEnabled2=true
    Module.Param2WriteEnabled=true
    Module.Param2WriteEnabled1=true
    Module.Param2WriteEnabled2=true
    Module.Param3WriteEnabled=true
    Module.Param3WriteEnabled1=true
    Module.Param3WriteEnabled2=true
    Write Parameter Index 1=true
    Write Parameter Index 2=true
   )
   &ScaleColor(
    Scale Alpha=&Multiply_Float(
     A=&FloatFromCurve(FloatCurve=DataInterface() {propertyValues {Curve {keys [
          <at 0 0 RCIM_Cubic>
          {arriveTangent=2, interpMode=RCIM_Cubic, leaveTangent=2, time=0.3, value=1}
          {interpMode=RCIM_Cubic, tangentMode=RCTM_User, time=0.5, value=1}
          {interpMode=RCIM_Cubic, tangentMode=RCTM_User, time=1, value=0}
         ]}}})
     B=&FloatFromCurve(
      FloatCurve=DataInterface() {propertyValues {Curve {keys [
          {arriveTangent=0.1443, interpMode=RCIM_Cubic, leaveTangent=0.1443, tangentMode=RCTM_User, time=1, value=0}
          {arriveTangent=0.6557, interpMode=RCIM_Cubic, leaveTangent=0.6557, tangentMode=RCTM_User, time=2, value=0.1}
          {arriveTangent=-0.0405, interpMode=RCIM_Cubic, leaveTangent=-0.0405, tangentMode=RCTM_User, time=3, value=1}
         ]}}}
      CurveIndex=$Emitter.Age
     )
    )
    Scale RGB=&MakeVectorFromLinearColorRGB(Color=&ColorFromCurve(ColorCurve=DataInterface() {propertyValues {
        AlphaCurve {keys [<at 0 1 RCIM_Cubic>,<at 1 1 RCIM_Cubic>]}
        BlueCurve {keys [<at 0.4624 5 RCIM_Cubic>,<at 0.7634 1 RCIM_Cubic>]}
        GreenCurve {keys [<at 0.4624 2.5 RCIM_Cubic>,<at 0.7634 1 RCIM_Cubic>]}
        RedCurve {keys [<at 0.4624 2.6431 RCIM_Cubic>,<at 0.7634 1 RCIM_Cubic>]}
       }}))
    ScaleA=true
    ScaleRGB=true
   )
   &CameraOffset(Camera Offset Amount=-10)
  ]
  rendererClasses [&NiagaraRibbonRendererProperties]
  renderers [
   {rendererClass=&NiagaraRibbonRendererProperties}
   {rendererClass=&NiagaraRibbonRendererProperties, rendererIndex=1}
   {rendererClass=&NiagaraRibbonRendererProperties, rendererIndex=2}
  ]
 }
 SmokeStreams_Upshort {
  bEnabled=false
  emitterUpdateScript [
   &EmitterState(Life Cycle Mode=::NewEnumerator1, Loop Behavior=::NewEnumerator1, Loop Duration Mode=::NewEnumerator1, UseLoopDelay=false)
   &SpawnRate(SpawnRate=50, Use Spawn Probability=false) {enabled=false}
   &FN_SpawnRibbon_v2(SpawnRate=200)
  ]
  particleSpawnScript [
   &InitializeParticle(Color Mode=::NewEnumerator0, Lifetime=0.1, UsePositionOffset=false, Write Material Random=true) {enabled=false}
   &InitializeRibbon(Color=<rgba 0.8333 0 1 0.5>, Lifetime=1.5, UsePositionOffset=false, Write Material Random=true)
   &AddVelocity()
   &SkeletalMeshLocation(
    Bone / Socket Index=5
    Bone Sampling Mode=::NewEnumerator5
    MaskSpawnGroup=false
    Sampled Position Offset=<0 0 -10>
    Sampled Position Offset Coordinate Space:ENiagaraSkelMeshTransforms=::NewEnumerator2
    Skeletal Mesh=$User.SkeletalMesh
   )
   &ApplyInitialForces()
  ]
  particleUpdateScript [
   &ParticleState()
   /SetVariables(
    Particles.Position001:NiagaraPosition=&ConvertVectorToPosition(
     Input Position=&Add_Vector(
      A=&ConvertPositionToVector(Input Position=$Particles.Position, Convert To Absolute Large World Vector=false)
      B=<7 0 10>
     )
     Vector Is In Absolute Large World Space:NiagaraBool=false
    )
    Particles.Position002:NiagaraPosition=&ConvertVectorToPosition(
     Input Position=&Add_Vector(
      A=&ConvertPositionToVector(Input Position=$Particles.Position, Convert To Absolute Large World Vector=false)
      B=<-11 0 5>
     )
     Vector Is In Absolute Large World Space:NiagaraBool=false
    )
    Particles.RibbonWidth001:NiagaraFloat=&Multiply_Float(A=$Particles.RibbonWidth, B=0.7)
    Particles.RibbonWidth002:NiagaraFloat=&Multiply_Float(A=$Particles.RibbonWidth, B=0.6)
   ) {isSetParameters=true, module=SetVariables_#3}
   &SolveForcesAndVelocity(Clamp Velocity=false, Limit Acceleration=false)
   &SkeletalMeshLocation(
    Bone / Socket Index=&SelectIntFromArray(Direct Array Index=3, Int Selection Array=DataInterface() {propertyValues {IntData [0,1]}})
    MaskSpawnGroup=false
    Mesh Sampling Type=::NewEnumerator5
    Position Sampling=::NewEnumerator1
    Skeletal Mesh=$User.SkeletalMesh
    Socket Sampling Mode=::NewEnumerator5
   ) {enabled=false, module=SkeletalMeshLocation001}
   &CurlNoiseForce(Pan Noise=false) {enabled=false, module=CurlNoiseForce001}
   &DynamicMaterialParameters(
    Index 0 Param 1=1
    Index 0 Param 2=1
    Index 0 Param 3=1
    Index 0 Param 4=1
    Index 1 Param 1=1
    Index 1 Param 2=1
    Index 2 Param 1=1
    Index 2 Param 2=1
    Index 2 Param 3=&FloatFromCurve(
     FloatCurve=DataInterface() {propertyValues {Curve {keys [
         {interpMode=RCIM_Cubic, tangentMode=RCTM_User, time=0, value=1}
         {interpMode=RCIM_Cubic, tangentMode=RCTM_User, time=0.1922, value=1}
         {tangentMode=RCTM_User, time=1, value=-1}
        ]}}}
     Scale Curve=0.5
    )
    Index 2 Param 4=&FloatFromCurve(FloatCurve=DataInterface() {propertyValues {Curve {keys [{interpMode=RCIM_Cubic, tangentMode=RCTM_User, time=0, value=2}, {interpMode=RCIM_Cubic, tangentMode=RCTM_User, time=0.5, value=1}]}}})
    Module.Param0WriteEnabled=true
    Module.Param0WriteEnabled1=true
    Module.Param0WriteEnabled2=true
    Module.Param0WriteEnabled3=true
    Module.Param1WriteEnabled=true
    Module.Param1WriteEnabled1=true
    Module.Param1WriteEnabled2=true
    Module.Param1WriteEnabled3=true
    Module.Param2WriteEnabled=true
    Module.Param2WriteEnabled1=true
    Module.Param2WriteEnabled2=true
    Module.Param2WriteEnabled3=true
    Module.Param3WriteEnabled=true
    Module.Param3WriteEnabled1=true
    Module.Param3WriteEnabled2=true
    Module.Param3WriteEnabled3=true
    Write Parameter Index 1=true
    Write Parameter Index 2=true
    Write Parameter Index 3=true
   ) {module=DynamicMaterialParameters001}
   &ScaleRibbonWidth(Ribbon Width Scale=&Multiply_Float(
     A=&Multiply_Float(
      A=&FloatFromCurve(FloatCurve=DataInterface() {propertyValues {Curve {keys [
           {interpMode=RCIM_Cubic, tangentMode=RCTM_User, time=0.1579, value=3.847}
           {arriveTangent=3.7274, interpMode=RCIM_Cubic, leaveTangent=3.7274, time=0.3145, value=5.967}
           {arriveTangent=-7.9244, interpMode=RCIM_Cubic, leaveTangent=-7.9244, time=0.4262, value=4.847}
           {arriveTangent=-4.9452, interpMode=RCIM_Cubic, leaveTangent=-4.9452, time=0.6889, value=3}
           {arriveTangent=0.6917, interpMode=RCIM_Cubic, leaveTangent=0.6917, tangentMode=RCTM_User, time=0.981, value=2.1033}
          ]}}})
      B=1
     )
     B=&FloatFromCurve(
      FloatCurve=DataInterface() {propertyValues {Curve {keys [
          {interpMode=RCIM_Cubic, tangentMode=RCTM_User, time=2, value=0.8}
          {interpMode=RCIM_Cubic, tangentMode=RCTM_User, time=3, value=0.8}
          <at 5 1 RCIM_Cubic>
         ]}}}
      CurveIndex=$Emitter.Age
     )
    ))
   &CurlNoiseForce(
    Noise Frequency=5
    Noise Strength=&FloatFromCurve(
     FloatCurve=DataInterface() {propertyValues {Curve {keys [{arriveTangent=-3.2172, interpMode=RCIM_Cubic, leaveTangent=-3.2172, tangentMode=RCTM_User, time=-0.0184, value=1.5941}, <at 1.0003 0.7389>]}}}
     CurveIndex=$Emitter.Age
     Scale Curve=1500
    )
    Pan Noise=true
    Pan Noise Field=<0 0 2>
    Random Seed:NiagaraInt32=$Particles.SpawnGroup
   ) {enabled=false}
   &DynamicMaterialParameters(
    Index 0 Param 1=1
    Index 0 Param 2=1
    Index 0 Param 3=1
    Index 0 Param 4=1
    Index 1 Param 1=1
    Index 1 Param 2=1
    Index 2 Param 1=0.2
    Index 2 Param 2=0.2
    Index 2 Param 3=0.2
    Index 2 Param 4=1
    Module.Param0WriteEnabled=true
    Module.Param0WriteEnabled1=true
    Module.Param0WriteEnabled2=true
    Module.Param1WriteEnabled=true
    Module.Param1WriteEnabled1=true
    Module.Param1WriteEnabled2=true
    Module.Param2WriteEnabled=true
    Module.Param2WriteEnabled1=true
    Module.Param2WriteEnabled2=true
    Module.Param3WriteEnabled=true
    Module.Param3WriteEnabled1=true
    Module.Param3WriteEnabled2=true
    Write Parameter Index 1=true
    Write Parameter Index 2=true
   ) {enabled=false}
   &ScaleColor(
    Scale Alpha=&Multiply_Float(
     A=&FloatFromCurve(FloatCurve=DataInterface() {propertyValues {Curve {keys [
          <at 0 0 RCIM_Cubic>
          {arriveTangent=2, interpMode=RCIM_Cubic, leaveTangent=2, time=0.3, value=1}
          {interpMode=RCIM_Cubic, tangentMode=RCTM_User, time=0.5, value=1}
          {interpMode=RCIM_Cubic, tangentMode=RCTM_User, time=1, value=0}
         ]}}})
     B=&FloatFromCurve(
      FloatCurve=DataInterface() {propertyValues {Curve {keys [
          {arriveTangent=0.1443, interpMode=RCIM_Cubic, leaveTangent=0.1443, tangentMode=RCTM_User, time=1, value=0}
          {arriveTangent=0.6557, interpMode=RCIM_Cubic, leaveTangent=0.6557, tangentMode=RCTM_User, time=2, value=0.1}
          {arriveTangent=-0.0405, interpMode=RCIM_Cubic, leaveTangent=-0.0405, tangentMode=RCTM_User, time=3, value=1}
         ]}}}
      CurveIndex=$Emitter.Age
     )
    )
    Scale RGB=&MakeVectorFromLinearColorRGB(Color=&ColorFromCurve(ColorCurve=DataInterface() {propertyValues {
        AlphaCurve {keys [<at 0 1 RCIM_Cubic>,<at 1 1 RCIM_Cubic>]}
        BlueCurve {keys [<at 0.4624 5 RCIM_Cubic>,<at 0.7634 1 RCIM_Cubic>]}
        GreenCurve {keys [<at 0.4624 2.5 RCIM_Cubic>,<at 0.7634 1 RCIM_Cubic>]}
        RedCurve {keys [<at 0.4624 2.6431 RCIM_Cubic>,<at 0.7634 1 RCIM_Cubic>]}
       }}))
    ScaleA=true
    ScaleRGB=true
   )
   &CameraOffset(Camera Offset Amount=-10)
  ]
  rendererClasses [&NiagaraRibbonRendererProperties]
  renderers [
   {rendererClass=&NiagaraRibbonRendererProperties}
   {rendererClass=&NiagaraRibbonRendererProperties, rendererIndex=1}
   {rendererClass=&NiagaraRibbonRendererProperties, rendererIndex=2}
  ]
 }
 SpinSplines {
  bEnabled=false
  emitterUpdateScript [
   &EmitterState()
   &SpawnBurst_Instantaneous(
    Spawn Count=40
    Spawn Probability=&MakeCustomFloatFromBool(Bool=$User.bIsFrontEndPreview, True Float=1, False Float=0)
    Use Spawn Probability=true
   ) {module=SpawnBurst_Instantaneous001}
  ]
  particleSpawnScript [
   /Niagara/Modules/Spawn/Initialization/InitializeParticle(
    Color=&MultiplyLinearColor(Linear Color=<rgba 1 1 1 1>, Scale Factor=<1 1 1 0.2>)
    Lifetime=&RandomRangeFloat(Minimum=0.55, Maximum=0.8)
    Mesh Scale=&RandomRangeVector(Minimum=<0.1 0.15 0.15>, Maximum=<0.4 0.2 0.2>)
    Module.Write Color=true
    Module.Write Lifetime=true
    Module.Write Mass=true
    Module.Write Material Random=true
    Module.Write Position=true
    Module.Write Scale=true
    Module.Write SpriteRotation=false
    Module.Write SpriteSize=false
    Sprite Size=<5 5>
   )
   &SkeletalMeshLocation(
    Apply Sampled Orientation to Mesh Particles=true
    Inherit Velocity=false
    MaskSpawnGroup=false
    Module.Override Randomness=true
    Module.Override Seed=false
    Skeletal Mesh=$User.SkeletalMesh
   )
   &InitialMeshOrientation(
    Orientation Axis=<0 0 -1>
    Rotation=&MakeVector(
     X=&RandomRangeFloat(Minimum=0, Maximum=1)
     Y=&FloatFromCurve(
      FloatCurve=DataInterface() {propertyValues {Curve {keys [
          <at 0 1 RCIM_Cubic>
          {arriveTangent=-5.2683, interpMode=RCIM_Cubic, leaveTangent=-5.2682, tangentMode=RCTM_User, time=0.4932, value=0}
          <at 1 -1>
         ]}}}
      CurveIndex=&RandomRangeFloat(Minimum=0, Maximum=1)
      Scale Curve=0.08
     )
     Z=0
    )
    Use Orientation Vector=true
    Use Rotation Vector=true
   ) {module=InitialMeshOrientation001}
  ]
  particleUpdateScript [
   &ParticleState()
   &ScaleColor(
    Scale Alpha=&FloatFromCurve(
     FloatCurve=DataInterface() {propertyValues {Curve {keys [
         {arriveTangent=0.0393, interpMode=RCIM_Cubic, leaveTangent=0.0393, tangentMode=RCTM_User, time=0.5, value=1.5}
         {arriveTangent=-0.0943, interpMode=RCIM_Cubic, leaveTangent=-0.0943, tangentMode=RCTM_User, time=1, value=0}
        ]}}}
     Scale Curve=0.8
    )
    Scale RGB=&VectorFromFloat(Value=&FloatFromCurve(FloatCurve=DataInterface() {propertyValues {Curve {keys [<at 0 0 RCIM_Cubic>,<at 0.1 1 RCIM_Cubic>,<at 0.75 0 RCIM_Cubic>]}}}))
    ScaleA=true
    ScaleRGB=false
   )
   &DynamicMaterialParameters(
    Index 0 Param 1=&FloatFromCurve(FloatCurve=DataInterface() {propertyValues {Curve {keys [{interpMode=RCIM_Cubic, tangentMode=RCTM_User, time=0, value=0.8}, {interpMode=RCIM_Cubic, tangentMode=RCTM_User, time=1, value=0.5}]}}})
    Index 0 Param 2=&FloatFromCurve(
     FloatCurve=DataInterface() {propertyValues {Curve {keys [
         {tangentMode=RCTM_User, time=0, value=1}
         {arriveTangent=-0.0405, interpMode=RCIM_Cubic, leaveTangent=-0.0405, tangentMode=RCTM_User, time=0.7009, value=1}
         {arriveTangent=-8.0089, interpMode=RCIM_Cubic, leaveTangent=-8.0089, tangentMode=RCTM_User, time=1, value=0}
        ]}}}
     Scale Curve=2
    )
    Index 0 Param 4=&RandomRangeFloat(Minimum=0, Maximum=1)
    Module.Param0WriteEnabled=true
    Module.Param1WriteEnabled=true
    Module.Param2WriteEnabled=false
    Module.Param3WriteEnabled=true
   )
   &UpdateMeshOrientation(
    Rotation Coordinate Space=::NewEnumerator3
    Rotation Rate=&FloatFromCurve(
     FloatCurve=DataInterface() {propertyValues {Curve {keys [
         {arriveTangent=7.0045, interpMode=RCIM_Cubic, leaveTangent=7.0045, tangentMode=RCTM_User, time=0, value=0.2}
         <at 0.25 1 RCIM_Cubic>
         <at 1 0.2 RCIM_Cubic>
        ]}}}
     Scale Curve=&RandomRangeFloat(Minimum=-0.3, Maximum=-0.16)
    )
    Rotation Vector=<1 0 0>
   ) {module=UpdateMeshOrientation001}
   &ScaleMeshSize(Scale Factor=&VectorFromFloat(Value=&FloatFromCurve(FloatCurve=DataInterface() {propertyValues {Curve {keys [
          {arriveTangent=0.0602, interpMode=RCIM_Cubic, leaveTangent=0.0602, tangentMode=RCTM_User, time=0, value=0}
          {arriveTangent=0.0697, interpMode=RCIM_Cubic, leaveTangent=0.0697, tangentMode=RCTM_User, time=0.8, value=1}
         ]}}})))
   &ScaleMeshSize(Scale Factor=&MakeVector(
     X=1
     Y=&FloatFromCurve(FloatCurve=DataInterface() {propertyValues {Curve {keys [
          {arriveTangent=2.4016, interpMode=RCIM_Cubic, leaveTangent=2.4016, tangentMode=RCTM_User, time=0, value=0.5}
          {interpMode=RCIM_Cubic, tangentMode=RCTM_User, time=0.6, value=1}
         ]}}})
     Z=&FloatFromCurve(FloatCurve=DataInterface() {propertyValues {Curve {keys [
          {arriveTangent=2.4016, interpMode=RCIM_Cubic, leaveTangent=2.4016, tangentMode=RCTM_User, time=0, value=0.5}
          {interpMode=RCIM_Cubic, tangentMode=RCTM_User, time=0.6, value=1}
         ]}}})
    )) {module=ScaleMeshSize001}
   &SolveForcesAndVelocity(Clamp Velocity=false, Limit Acceleration=false)
  ]
  rendererClasses [&NiagaraMeshRendererProperties]
  renderers [{rendererClass=&NiagaraMeshRendererProperties, rendererIndex=0}]
 }
 Surround_modified {
  bEnabled=false
  emitterUpdateScript [
   &EmitterState()
   &SpawnRate(SpawnRate=20, Use Spawn Probability=false)
  ]
  particleSpawnScript [
   /SetVariables(Particles.Random=&RandomRangeFloat(Minimum=0, Maximum=1)) {isSetParameters=true, module=SetVariables_#1}
   &InitializeParticle(
    Color=<rgba 0 0 0 1>
    Lifetime=&Lerp_Float(A=0.3, B=0.5, Alpha=$Particles.Random)
    Module.Write Color=true
    Module.Write Lifetime=true
    Module.Write Mass=true
    Module.Write Material Random=true
    Module.Write Position=false
    Module.Write Scale=false
    Module.Write SpriteRotation=false
    Module.Write SpriteSize=true
    Position=&SimulationPosition()
    Sprite Size=<30 100>
   )
   &CameraOffset(Camera Offset Amount=&RandomRangeFloat(Minimum=0, Maximum=35))
   &DynamicMaterialParameters(
    Index 0 Param 1=&Lerp_Float(A=0, B=1, Alpha=$Particles.Random)
    Index 0 Param 2=&Lerp_Float(A=0, B=0.5, Alpha=$Particles.Random)
    Index 0 Param 3=-3
    Index 0 Param 4=5
    Module.Param0WriteEnabled=true
    Module.Param1WriteEnabled=true
    Module.Param2WriteEnabled=true
    Module.Param3WriteEnabled=true
   )
   &CylinderLocation(
    Cylinder Radius=35
    MaskSpawnGroup=false
    Module.Enable Lathe Profile=false
    Module.Override Randomness=false
    Module.Surface Only=false
    Offset=<0 0 65>
   )
  ]
  particleUpdateScript [
   &ParticleState()
   &ScaleSpriteSize(Scale Factor=&MakeVector2D(
     X=&FloatFromCurve(FloatCurve=DataInterface() {propertyValues {Curve {keys [<at 0 1.5>,<at 1 0.5>]}}})
     Y=&FloatFromCurve(FloatCurve=DataInterface() {propertyValues {Curve {keys [<at 0 1>,<at 1 2>]}}})
    ))
   &SolveForcesAndVelocity(Clamp Velocity=false, Limit Acceleration=false)
  ]
  rendererClasses [&NiagaraSpriteRendererProperties]
  renderers [{rendererClass=&NiagaraSpriteRendererProperties}]
 }
 Wave {
  bEnabled=false
  emitterUpdateScript [
   &EmitterState()
   &SpawnBurst_Instantaneous(Spawn Count=1, Use Spawn Probability=false)
  ]
  particleSpawnScript [
   /Niagara/Modules/Spawn/Initialization/V2/InitializeParticle(
    Color Mode=::NewEnumerator0
    Lifetime Max=1.25
    Lifetime Min=1.15
    Lifetime Mode=::NewEnumerator1
    Mesh Scale Mode=::NewEnumerator3
    Mesh Uniform Scale=&RandomRangeFloat(Minimum=1, Maximum=1)
    UsePositionOffset=false
    Write Material Random=true
   )
   &InitialMeshOrientation(Orientation Axis=<0 0 1>, Orientation Vector=<0 0 1>, Rotation=<0 0.25 0>, Use Orientation Vector=true, Use Rotation Vector=true)
   &InitialMeshOrientation(
    Orientation Axis=<0 0 0>
    Orientation Vector=<0 0 0>
    Rotation=&RandomRangeVector(Minimum=<0 0 0>, Maximum=<1 0 0>)
    Use Orientation Vector=false
    Use Rotation Vector=true
   ) {module=InitialMeshOrientation001}
   &CameraOffset(Camera Offset Amount=-10)
   &SkeletalMeshLocation(
    Inherit Velocity=false
    MaskSpawnGroup=false
    Mesh Sampling Type=::NewEnumerator5
    Skeletal Mesh=$User.SkeletalMesh
    Socket Sampling Mode=::NewEnumerator5
   )
   &AccelerationForce(Acceleration=<0 0 15>)
   &ApplyInitialForces()
  ]
  particleUpdateScript [
   &SolveForcesAndVelocity(Clamp Velocity=false, Limit Acceleration=false)
   &ParticleState()
   &ScaleColor(
    Linear Color Curve=DataInterface() {propertyValues {
      AlphaCurve {keys [<at 0 0>,<at 0.0695 0.5>,<at 0.3576 0.5>,<at 1 0>]}
      BlueCurve {keys [<at 0.0283 1>,<at 0.1987 1>,<at 0.3149 1>,<at 0.8 1>]}
      GreenCurve {keys [<at 0.0283 0>,<at 0.1987 0.8167>,<at 0.3149 0.3012>,<at 0.8 0.0414>]}
      RedCurve {keys [<at 0.0283 0>,<at 0.1987 0>,<at 0.3149 0>,<at 0.8 0.5686>]}
     }}
    Scale Mode=::NewEnumerator2
   ) {module=ScaleColor002}
   &DynamicMaterialParameters(
    Index 0 Param 1=&FloatFromCurve(FloatCurve=DataInterface() {propertyValues {Curve {keys [{arriveTangent=-4.0141, interpMode=RCIM_Cubic, leaveTangent=-4.0141, tangentMode=RCTM_User, time=0, value=0.6}, <at 1 -0.7>]}}})
    Index 0 Param 2=&FloatFromCurve(FloatCurve=DataInterface() {propertyValues {Curve {keys [{arriveTangent=-0.5818, interpMode=RCIM_Cubic, leaveTangent=-0.5818, tangentMode=RCTM_User, time=0, value=1}, <at 1 0.9>]}}})
    Index 0 Param 3=&FloatFromCurve(FloatCurve=DataInterface() {propertyValues {Curve {keys [<at 0 0>,<at 0.15 1>,<at 1 1.2>]}}})
    Module.Param0WriteEnabled=true
    Module.Param1WriteEnabled=true
    Module.Param2WriteEnabled=true
    Module.Param3WriteEnabled=false
   )
   &ScaleColor(
    Scale Alpha=0.5
    Scale RGB=&VectorFromFloat(Value=&FloatFromCurve(FloatCurve=DataInterface() {propertyValues {Curve {keys [<at 0 10>,<at 0.15 3>,<at 1 1>]}}}))
    ScaleA=true
    ScaleRGB=true
   ) {module=ScaleColor001}
   &ScaleMeshSize(Scale Factor=&VectorFromFloat(Value=&FloatFromCurve(FloatCurve=DataInterface() {propertyValues {Curve {keys [
          {arriveTangent=12.9962, interpMode=RCIM_Cubic, leaveTangent=12.9962, tangentMode=RCTM_User, time=0, value=0}
          {arriveTangent=2, interpMode=RCIM_Cubic, leaveTangent=2, tangentMode=RCTM_User, time=0.1742, value=1.2171}
          <at 1 2>
         ]}}})))
  ]
  rendererClasses [&NiagaraMeshRendererProperties]
  renderers [{rendererClass=&NiagaraMeshRendererProperties, rendererIndex=0}]
 }
}
summary {
 systemName=NS_Emote_Origin_Prison2
 emitters [
  {emitterName=BurstEmbers, rendererClasses [&NiagaraSpriteRendererProperties]}
  {bEnabled=false, emitterName=SmokeLoop, rendererClasses [&NiagaraSpriteRendererProperties]}
  {bEnabled=false, emitterName=ShieldDecals, rendererClasses [&NiagaraSpriteRendererProperties]}
  {bEnabled=false, emitterName=SpinSplines, rendererClasses [&NiagaraMeshRendererProperties]}
  {bEnabled=false, emitterName=BlackDecals, rendererClasses [&NiagaraSpriteRendererProperties]}
  {emitterName=Digital_Bits, simTarget=GPUComputeSim, rendererClasses [&NiagaraSpriteRendererProperties]}
  {emitterName=ChromaSprites, rendererClasses [&NiagaraSpriteRendererProperties]}
  {emitterName=Digital_Bits001, simTarget=GPUComputeSim, rendererClasses [&NiagaraSpriteRendererProperties]}
  {emitterName=ChromaSprites001, rendererClasses [&NiagaraSpriteRendererProperties]}
  {
   bEnabled=false
   emitterName=Beam
   rendererClasses [&NiagaraMeshRendererProperties, &NiagaraLightRendererProperties]
  }
  {bEnabled=false, emitterName=Surround_modified, rendererClasses [&NiagaraSpriteRendererProperties]}
  {bEnabled=false, emitterName=Blocks, rendererClasses [&NiagaraSpriteRendererProperties]}
  {emitterName=FLASH, rendererClasses [&NiagaraSpriteRendererProperties]}
  {bEnabled=false, emitterName=Wave, rendererClasses [&NiagaraMeshRendererProperties]}
  {bEnabled=false, emitterName=SmokeStreams_Up, rendererClasses [&NiagaraRibbonRendererProperties]}
  {bEnabled=false, emitterName=SmokeStreams_Upshort, rendererClasses [&NiagaraRibbonRendererProperties]}
 ]
 userVariables [
  User.bIsFrontEndPreview {type=NiagaraBool, defaultValue=true}
  User.Color_01 {type=LinearColor, description="", defaultValue {struct=LinearColor, value=<rgba 0.5701 0 1 1>}}
  User.Color_02 {type=LinearColor, description="", defaultValue {struct=LinearColor, value=<rgba 1 0.6238 0 1>}}
  User.RibbonIDOffset {type=NiagaraInt32}
  User.SkeletalMesh {
   type=NiagaraDataInterfaceSkeletalMesh
   defaultValue {
    struct=NiagaraExt_VariableValue_DataInterface
    value {dataInterface=&self:NiagaraDataInterfaceSkeletalMesh_2}
   }
  }
 ]
}
