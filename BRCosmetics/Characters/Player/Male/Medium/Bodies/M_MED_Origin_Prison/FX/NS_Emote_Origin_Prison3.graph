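# Niagara effect for an emote on the medium male Origin Prison character, layering five timed sprite and ribbon emitters.
# Hands_Pulse pins growing, color-shifting sprites to two bones of the user skeletal mesh.
# Two SmokeStreams emitters trail ribbons from another bone, each with three renderers plus offset position and width copies, shaped by emitter-age curves.
# Delayed Blocks and Surround_modified add rising sprites from an expanding squashed sphere and a tiny offset cylinder.
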NiagaraSystem /BRCosmetics/Characters/Player/Male/Medium/Bodies/M_MED_Origin_Prison/FX/NS_Emote_Origin_Prison3
dependencies {
 usedDataInterfaces [
  &NiagaraDataInterfaceArrayInt32
  &NiagaraDataInterfaceCurve
  &NiagaraDataInterfaceColorCurve
  &NiagaraDataInterfaceVector2DCurve
  &NiagaraDataInterfaceVectorCurve
  &NiagaraDataInterfaceArrayFloat3
 ]
 usedDynamicInputs [
  &NormalizeVector
  &Multiply_Vector
  &Multiply_VectorByFloat
 ]
 usedModules [
  &SystemState
  &self:SystemSpawnScript.SystemScriptSource.SystemScriptGraph.NiagaraNodeAssignment_5.SetVariables_#1
  &InterpolateOverTime
  &self:SystemSpawnScript.SystemScriptSource.SystemScriptGraph.NiagaraNodeAssignment_1.SetVariables_#2
  &self:SystemSpawnScript.SystemScriptSource.SystemScriptGraph.NiagaraNodeAssignment_0.SetVariables_#3
 ]
}
emitters {
 Blocks {
  emitterUpdateScript [
   &EmitterState(
    Inactive Response=::NewEnumerator1
    Life Cycle Mode=::NewEnumerator1
    Loop Behavior=::NewEnumerator1
    Loop Delay=1.6704
    Loop Duration=1.9296
    UseLoopDelay=true
   )
   &SpawnBurst_Instantaneous(Spawn Count=20, Use Spawn Probability=false)
   &SpawnRate(
    SpawnRate=&FloatFromCurve(FloatCurve=DataInterface() {propertyValues {Curve {keys [<at 0.1 12>,<at 0.6 28>]}}}, CurveIndex=$Emitter.NormalizedLoopAge)
    Use Spawn Probability=false
   )
   &SpawnBurst_Instantaneous(Spawn Count=20, Spawn Time=0.4296, Use Spawn Probability=false, UseLoopCountLimit=false) {module=SpawnBurst_Instantaneous001}
   &SpawnBurst_Instantaneous(Spawn Count=20, Spawn Time=0.2834, Use Spawn Probability=false, UseLoopCountLimit=false) {module=SpawnBurst_Instantaneous002}
  ]
  particleSpawnScript [
   /SetVariables(Particles.Random=&RandomRangeFloat(Minimum=0, Maximum=1)) {isSetParameters=true, module=SetVariables_#4}
   /Niagara/Modules/Spawn/Initialization/InitializeParticle(
    Color=&LerpLinearColors(StartColor=<rgba 0 0.0914 0.1 1>, EndColor=<rgba 0.02 0.0186 0.0115 1>, LerpFactor=&UniformAOrBFloat(A=0))
    Lifetime=&Lerp_Float(A=0.3, B=0.5, Alpha=$Particles.Random)
    Module.Write Color=true
    Module.Write Lifetime=true
    Module.Write Mass=true
    Module.Write Material Random=true
    Module.Write Position=false
    Module.Write Scale=false
    Module.Write SpriteRotation=false
    Module.Write SpriteSize=true
    Position=&SimulationPosition()
    Sprite Size=&Lerp_Vector2(A=<8 12.5>, B=<5 10>, Alpha=$Particles.Random)
   )
   &FN_SetSpawnGroup_v2(Number of Spawn Groups=17) {enabled=false}
   &DynamicMaterialParameters(
    Index 0 Param 1=&FloatFromCurve(FloatCurve=DataInterface() {propertyValues {Curve {keys [<at 0 0.3>,<at 0.6 0.1>,<at 1 0.1>]}}}, CurveIndex=$Emitter.NormalizedLoopAge)
    Index 0 Param 2=&Lerp_Float(A=1.5, B=2.5, Alpha=$Particles.Random)
    Index 0 Param 3=&Lerp_Float(A=0, B=1, Alpha=$Particles.Random)
    Index 0 Param 4=1
    Module.Param0WriteEnabled=true
    Module.Param1WriteEnabled=true
    Module.Param2WriteEnabled=true
    Module.Param3WriteEnabled=true
   )
   &CameraOffset(Camera Offset Amount=20) {enabled=false}
   &SphereLocation(
    MaskSpawnGroup=false
    Non Uniform Scale=<1 0.1 0.5>
    Offset=<0 15 100>
    Override Seed=false
    Sphere Coordinate Space=::World
    Sphere Radius=&FloatFromCurve(
     FloatCurve=DataInterface() {propertyValues {Curve {keys [
         {interpMode=RCIM_Cubic, tangentMode=RCTM_User, time=0, value=0}
         {arriveTangent=19.8, interpMode=RCIM_Cubic, time=0.02, value=1}
         {arriveTangent=6.3711, interpMode=RCIM_Cubic, leaveTangent=6.3711, tangentMode=RCTM_User, time=1, value=10}
        ]}}}
     CurveIndex=$Emitter.NormalizedLoopAge
     Scale Curve=2
    )
    Surface Only=true
    Surface Only Band Thickness=0.1
    UseNonUniformScale=true
    UseOffset=true
    UseSphereOrientation=false
   )
   /SetVariables(
    Particles.Color=&MultiplyLinearColor(
     Linear Color=$Particles.Initial.Color
     Scale Factor=&MakeFloat4FromVectorAndFloat(
      "Vector (RGB)"=&VectorFromCurve(
       VectorCurve=DataInterface() {propertyValues {
         XCurve {keys [{tangentMode=RCTM_User, time=0, value=1}, {tangentMode=RCTM_User, time=1, value=1}]}
         YCurve {keys [{tangentMode=RCTM_User, time=0, value=1}, {tangentMode=RCTM_User, time=1, value=2}]}
         ZCurve {keys [{tangentMode=RCTM_User, time=0, value=1}, {tangentMode=RCTM_User, time=1, value=2}]}
        }}
       CurveIndex=$Emitter.NormalizedLoopAge
       Scale Curve=<1 0.2 1>
      )
      "Float (Alpha)"=10
     )
    )
    Particles.SpriteSize=&Multiply_Vector2D(
     A=$Particles.Initial.SpriteSize
     B=&Vector2DFromCurve(
      Vector2Curve=DataInterface() {propertyValues {
        XCurve {keys [
          <at 0 0.5 RCIM_Constant>
          <at 0.2 0.2819 RCIM_Constant>
          {arriveTangent=1.7446, interpMode=RCIM_Cubic, leaveTangent=1.7446, tangentMode=RCTM_User, time=1, value=1}
         ]}
        YCurve {keys [
          <at 0 0.7432 RCIM_Constant>
          <at 0.285 0.5 RCIM_Constant>
          <at 0.4437 1 RCIM_Constant>
          <at 0.7084 0.825 RCIM_Constant>
          {arriveTangent=1.565, interpMode=RCIM_Cubic, leaveTangent=1.565, tangentMode=RCTM_User, time=1, value=1}
         ]}
       }}
      CurveIndex=$Emitter.LoopedAge
      ScaleCurve=<1.5 2>
     )
    )
   ) {isSetParameters=true}
   &AddVelocity(Velocity=&VectorFromCurve(
     VectorCurve=DataInterface() {propertyValues {XCurve {keys [<at 0 1>,<at 1 0>]}, YCurve {keys [<at 0 1>,<at 1 0>]}, ZCurve {keys [<at 0.1 0.5>,<at 0.5 1>]}}}
     CurveIndex=$System.NormalizedLoopAge
    ))
  ]
  particleUpdateScript [
   &ParticleState()
   &AccelerationForce(Acceleration=&Lerp_Vector(A=<0 0 500>, B=<0 0 50>, Alpha=$Particles.Random))
   &ScaleSpriteSize(
    Scale Factor=&RandomRangeVector2D(Minimum=<1 1>, Maximum=<2 5>, Evaluation Type=::NewEnumerator1)
    Scale Sprite Size Mode=::NewEnumerator2
   )
   &ScaleColor(
    Scale Alpha=&FloatFromCurve(
     FloatCurve=DataInterface() {propertyValues {Curve {keys [
         {arriveTangent=-0.1617, interpMode=RCIM_Cubic, leaveTangent=-0.1617, tangentMode=RCTM_User, time=0, value=0}
         {arriveTangent=-0.0647, interpMode=RCIM_Cubic, leaveTangent=-0.0647, tangentMode=RCTM_User, time=0.6247, value=1}
         <at 1 0>
        ]}}}
     CurveIndex=$Emitter.NormalizedLoopAge
     Scale Curve=0.1
    )
    Scale RGB=&VectorFromFloat(Value=&FloatFromCurve(FloatCurve=DataInterface() {propertyValues {Curve {keys [<at 0 15>,<at 0.6129 0.434>,<at 1 0>]}}}))
    ScaleA=true
    ScaleRGB=true
   )
   &SolveForcesAndVelocity(Clamp Velocity=false, Limit Acceleration=false)
  ]
  rendererClasses [&NiagaraSpriteRendererProperties]
  renderers [{rendererClass=&NiagaraSpriteRendererProperties}]
 }
 Hands_Pulse {
  emitterUpdateScript [
   &EmitterState(Life Cycle Mode=::NewEnumerator1, Loop Behavior=::NewEnumerator1, Loop Delay=0.4, Loop Duration=1.5, UseLoopDelay=true)
   &SpawnBurst_Instantaneous(Spawn Count=2, Use Spawn Probability=false, UseLoopCountLimit=false)
  ]
  particleSpawnScript [/Niagara/Modules/Spawn/Initialization/V2/InitializeParticle(
    Color=<rgba 1 1 1 0.4>
    Lifetime Max=2.2
    Lifetime Min=1.5
    Lifetime Mode=::NewEnumerator1
    Sprite Size Mode=::NewEnumerator3
    Uniform Sprite Size=100
    UsePositionOffset=false
    Write Material Random=true
   )]
  particleUpdateScript [
   &ParticleState(Kill Particles When Lifetime Has Elapsed=false)
   &SkeletalMeshLocation(
    Bone / Socket Index=&SelectIntFromArray(
     Array Sampling Mode=::NewEnumerator1
     Direct Array Index=?
     Int Selection Array=DataInterface() {propertyValues {IntData [17,45]}}
     Random Seed=0
    )
    Bone Sampling Mode=::NewEnumerator6
    MaskSpawnGroup=false
    Skeletal Mesh=$User.SkeletalMesh
   )
   &DynamicMaterialParameters(
    Index 0 Param 1=&FloatFromCurve(FloatCurve=DataInterface() {propertyValues {Curve {keys [<at 0 1 RCIM_Constant>,<at 0.3 0.1>]}}})
    Index 0 Param 2=&FloatFromCurve(FloatCurve=DataInterface() {propertyValues {Curve {keys [<at 0 0>,<at 0.5 0.1 RCIM_Constant>,<at 0.6 1>]}}})
    Module.Param0WriteEnabled=true
    Module.Param0WriteEnabled1=true
    Module.Param1WriteEnabled=true
    Module.Param1WriteEnabled1=true
    Module.Param2WriteEnabled=true
    Module.Param2WriteEnabled1=true
    Module.Param3WriteEnabled=true
    Module.Param3WriteEnabled1=true
    Write Parameter Index 1=true
   )
   /SetVariables(Particles.SpriteRotation:NiagaraFloat=&Waveform("Waveform [1]"=::NewEnumerator5, "[1] Phase Scale"=110, "[1] Frequency"=100, "[1] Amplitude Min/Max"=<0 300>, "[1] Bias"=1)) {isSetParameters=true}
   &ScaleColor(
    Linear Color Curve=DataInterface() {propertyValues {
      AlphaCurve {keys [<at 0.4208 1 RCIM_Cubic>, {arriveTangent=-2.8965, interpMode=RCIM_Cubic, leaveTangent=-2.8965, time=0.5524, value=0.2}, <at 0.697 0.2>]}
      BlueCurve {keys [<at 0.4024 1 RCIM_Cubic>, <at 0.5118 0.3636>, {interpMode=RCIM_Cubic, leaveTangent=-2.6407, time=0.7604, value=5}]}
      GreenCurve {keys [<at 0.4024 0.1369 RCIM_Cubic>, <at 0.5118 0.5874>, {interpMode=RCIM_Cubic, leaveTangent=1.8695, time=0.7604, value=3.8398}]}
      RedCurve {keys [<at 0.4024 0.821 RCIM_Cubic>, <at 0.5118 1>, {interpMode=RCIM_Cubic, leaveTangent=0.7425, time=0.7604, value=0.8145}]}
     }}
    Scale Mode=::NewEnumerator2
   )
   /SetVariables(Particles.Color001:LinearColor=&ColorFromCurve(ColorCurve=DataInterface() {propertyValues {
       AlphaCurve {keys [<at 0.343 1>,<at 0.4208 1 RCIM_Cubic>,<at 0.5524 1 RCIM_Cubic>,<at 0.6926 1>]}
       BlueCurve {keys [<at 0.2817 1 RCIM_Cubic>,<at 0.3307 0.0302>,<at 0.4289 0.5>,<at 0.5037 0>,<at 0.623 0 RCIM_Cubic>,<at 0.681 0>,<at 0.8351 0.2917>]}
       GreenCurve {keys [
         <at 0.2817 0.1369 RCIM_Cubic>
         <at 0.3307 0>
         <at 0.4289 0.1851>
         <at 0.5037 1>
         {arriveTangent=-5.6402, interpMode=RCIM_Cubic, leaveTangent=-5.6402, time=0.623, value=1}
         <at 0.681 0>
         <at 0.8351 0.2917>
        ]}
       RedCurve {keys [
         <at 0.2817 0.821 RCIM_Cubic>
         <at 0.3307 0.2>
         <at 0.4289 0.0802>
         <at 0.5037 0.92>
         {arriveTangent=-2.369, interpMode=RCIM_Cubic, leaveTangent=-2.369, time=0.623, value=0.045}
         <at 0.681 0.5>
         <at 0.8351 2>
        ]}
      }})) {isSetParameters=true}
   &ScaleSpriteSize(Uniform Curve Sprite Scale=DataInterface() {propertyValues {Curve {keys [<at 0 1 RCIM_Cubic>, {arriveTangent=2.0873, interpMode=RCIM_Cubic, leaveTangent=2.0873, tangentMode=RCTM_User, time=1, value=4}]}}})
   /SetVariables(Particles.SpriteSize001:Vector2f=&Vector2DFromFloat(Value=&FloatFromCurve(
      FloatCurve=DataInterface() {propertyValues {Curve {keys [
          {arriveTangent=-2.5949, interpMode=RCIM_Cubic, leaveTangent=-2.5949, tangentMode=RCTM_User, time=0, value=0.5}
          {interpMode=RCIM_Cubic, tangentMode=RCTM_User, time=0.18, value=0.1}
          {arriveTangent=-0.2, interpMode=RCIM_Cubic, leaveTangent=-0.2, time=0.3, value=0.1}
          {arriveTangent=0.4015, interpMode=RCIM_Cubic, leaveTangent=0.4015, tangentMode=RCTM_User, time=1, value=0.3}
         ]}}}
      Scale Curve=200
     ))) {isSetParameters=true}
  ]
  rendererClasses [&NiagaraSpriteRendererProperties]
  renderers [
   {rendererClass=&NiagaraSpriteRendererProperties}
   {rendererClass=&NiagaraSpriteRendererProperties, rendererIndex=1}
  ]
 }
 SmokeStreams_Up {
  emitterUpdateScript [
   &EmitterState(Life Cycle Mode=::NewEnumerator1, Loop Behavior=::NewEnumerator1, Loop Duration Mode=::NewEnumerator1, UseLoopDelay=true)
   &SpawnRate(SpawnRate=50, Use Spawn Probability=false) {enabled=false}
   &FN_SpawnRibbon_v2(SpawnRate=200)
  ]
  particleSpawnScript [
   &InitializeParticle(Color Mode=::NewEnumerator0, Lifetime=0.1, UsePositionOffset=false, Write Material Random=true) {enabled=false}
   &InitializeRibbon(
    Color=<rgba 0.9167 0.5 1 0.2>
    Lifetime=&FloatFromCurve(
     FloatCurve=DataInterface() {propertyValues {Curve {keys [{interpMode=RCIM_Cubic, tangentMode=RCTM_User, time=0.5, value=0.2}, {interpMode=RCIM_Cubic, tangentMode=RCTM_User, time=2, value=1}]}}}
     CurveIndex=$Emitter.Age
     Scale Curve=2
    )
    Ribbon Width=15
    UsePositionOffset=false
    Write Material Random=true
   )
   &AddVelocity()
   &SkeletalMeshLocation(
    Bone / Socket Index=5
    Bone Sampling Mode=::NewEnumerator5
    MaskSpawnGroup=false
    Sampled Position Offset=<0 0 -10>
    Sampled Position Offset Coordinate Space:ENiagaraSkelMeshTransforms=::NewEnumerator2
    Skeletal Mesh=$User.SkeletalMesh
   )
   &ApplyInitialForces()
  ]
  particleUpdateScript [
   &ParticleState()
   /SetVariables(
    Particles.Position001:NiagaraPosition=&ConvertVectorToPosition(
     Input Position=&Add_Vector(
      A=&ConvertPositionToVector(Input Position=$Particles.Position, Convert To Absolute Large World Vector=false)
      B=<8 0 10>
     )
     Vector Is In Absolute Large World Space:NiagaraBool=false
    )
    Particles.Position002:NiagaraPosition=&ConvertVectorToPosition(
     Input Position=&Add_Vector(
      A=&ConvertPositionToVector(Input Position=$Particles.Position, Convert To Absolute Large World Vector=false)
      B=<-9 0 5>
     )
     Vector Is In Absolute Large World Space:NiagaraBool=false
    )
    Particles.RibbonWidth001:NiagaraFloat=&Multiply_Float(A=$Particles.RibbonWidth, B=0.7)
    Particles.RibbonWidth002:NiagaraFloat=&Multiply_Float(A=$Particles.RibbonWidth, B=0.6)
   ) {isSetParameters=true, module=SetVariables_#5}
   &SolveForcesAndVelocity(Clamp Velocity=false, Limit Acceleration=false)
   &SkeletalMeshLocation(
    Bone / Socket Index=&SelectIntFromArray(Direct Array Index=3, Int Selection Array=DataInterface() {propertyValues {IntData [0,1]}})
    MaskSpawnGroup=false
    Mesh Sampling Type=::NewEnumerator5
    Position Sampling=::NewEnumerator1
    Skeletal Mesh=$User.SkeletalMesh
    Socket Sampling Mode=::NewEnumerator5
   ) {enabled=false, module=SkeletalMeshLocation001}
   &CurlNoiseForce(Pan Noise=false) {enabled=false, module=CurlNoiseForce001}
   &DynamicMaterialParameters(
    Index 0 Param 1=1
    Index 0 Param 2=1
    Index 0 Param 3=1
    Index 0 Param 4=1
    Index 1 Param 1=1
    Index 1 Param 2=1
    Index 2 Param 1=1
    Index 2 Param 2=1
    Index 2 Param 3=&FloatFromCurve(FloatCurve=DataInterface() {propertyValues {Curve {keys [
         {arriveTangent=-2.9047, interpMode=RCIM_Cubic, leaveTangent=-2.9047, tangentMode=RCTM_User, time=0, value=1}
         {arriveTangent=-1, interpMode=RCIM_Cubic, leaveTangent=-1, time=0.5, value=-1}
         {interpMode=RCIM_Cubic, tangentMode=RCTM_User, time=1, value=0}
        ]}}})
    Index 2 Param 4=&FloatFromCurve(FloatCurve=DataInterface() {propertyValues {Curve {keys [
         {arriveTangent=-2.1237, interpMode=RCIM_Cubic, leaveTangent=-2.1237, tangentMode=RCTM_User, time=0, value=1}
         {interpMode=RCIM_Cubic, tangentMode=RCTM_User, time=1, value=1}
        ]}}})
    Module.Param0WriteEnabled=true
    Module.Param0WriteEnabled1=true
    Module.Param0WriteEnabled2=true
    Module.Param0WriteEnabled3=true
    Module.Param1WriteEnabled=true
    Module.Param1WriteEnabled1=true
    Module.Param1WriteEnabled2=true
    Module.Param1WriteEnabled3=true
    Module.Param2WriteEnabled=true
    Module.Param2WriteEnabled1=true
    Module.Param2WriteEnabled2=true
    Module.Param2WriteEnabled3=true
    Module.Param3WriteEnabled=true
    Module.Param3WriteEnabled1=true
    Module.Param3WriteEnabled2=true
    Module.Param3WriteEnabled3=true
    Write Parameter Index 1=true
    Write Parameter Index 2=true
    Write Parameter Index 3=true
   ) {module=DynamicMaterialParameters001}
   &ScaleRibbonWidth(Ribbon Width Scale=&Multiply_Float(
     A=&Multiply_Float(
      A=&FloatFromCurve(FloatCurve=DataInterface() {propertyValues {Curve {keys [
           {interpMode=RCIM_Cubic, tangentMode=RCTM_User, time=0.2, value=5}
           {arriveTangent=-7.3024, interpMode=RCIM_Cubic, leaveTangent=-7.3024, time=0.3907, value=4.7777}
           {arriveTangent=-2.2599, interpMode=RCIM_Cubic, leaveTangent=-2.2599, time=0.5821, value=2.2096}
           {arriveTangent=-4.0912, interpMode=RCIM_Cubic, leaveTangent=-4.0912, time=0.8163, value=3.8158}
           {arriveTangent=0.6917, interpMode=RCIM_Cubic, leaveTangent=0.6917, tangentMode=RCTM_User, time=1, value=0.5}
          ]}}})
      B=1
     )
     B=&FloatFromCurve(
      FloatCurve=DataInterface() {propertyValues {Curve {keys [
          {interpMode=RCIM_Cubic, tangentMode=RCTM_User, time=1, value=1}
          {interpMode=RCIM_Cubic, tangentMode=RCTM_User, time=2.5, value=0.5}
          <at 4 1 RCIM_Cubic>
         ]}}}
      CurveIndex=$Emitter.Age
     )
    ))
   &CurlNoiseForce(
    Noise Frequency=5
    Noise Strength=&FloatFromCurve(
     FloatCurve=DataInterface() {propertyValues {Curve {keys [{arriveTangent=-3.2172, interpMode=RCIM_Cubic, leaveTangent=-3.2172, tangentMode=RCTM_User, time=-0.0184, value=1.5941}, <at 1.0003 0.7389>]}}}
     CurveIndex=$Emitter.Age
     Scale Curve=1500
    )
    Pan Noise=true
    Pan Noise Field=<0 0 2>
    Random Seed:NiagaraInt32=$Particles.SpawnGroup
   ) {enabled=false}
   &DynamicMaterialParameters(
    Index 0 Param 1=1
    Index 0 Param 2=1
    Index 0 Param 3=1
    Index 0 Param 4=1
    Index 1 Param 1=1
    Index 1 Param 2=1
    Index 2 Param 1=0.2
    Index 2 Param 2=0.2
    Index 2 Param 3=0.2
    Index 2 Param 4=1
    Module.Param0WriteEnabled=true
    Module.Param0WriteEnabled1=true
    Module.Param0WriteEnabled2=true
    Module.Param1WriteEnabled=true
    Module.Param1WriteEnabled1=true
    Module.Param1WriteEnabled2=true
    Module.Param2WriteEnabled=true
    Module.Param2WriteEnabled1=true
    Module.Param2WriteEnabled2=true
    Module.Param3WriteEnabled=true
    Module.Param3WriteEnabled1=true
    Module.Param3WriteEnabled2=true
    Write Parameter Index 1=true
    Write Parameter Index 2=true
   )
   &ScaleColor(
    Scale Alpha=&Multiply_Float(
     A=&FloatFromCurve(FloatCurve=DataInterface() {propertyValues {Curve {keys [
          <at 0 0 RCIM_Cubic>
          {arriveTangent=2, interpMode=RCIM_Cubic, leaveTangent=2, time=0.3, value=1}
          {interpMode=RCIM_Cubic, tangentMode=RCTM_User, time=0.5, value=1}
          {interpMode=RCIM_Cubic, tangentMode=RCTM_User, time=1, value=0}
         ]}}})
     B=&FloatFromCurve(
      FloatCurve=DataInterface() {propertyValues {Curve {keys [
          {arriveTangent=0.1443, interpMode=RCIM_Cubic, leaveTangent=0.1443, tangentMode=RCTM_User, time=1, value=0}
          {arriveTangent=0.6557, interpMode=RCIM_Cubic, leaveTangent=0.6557, tangentMode=RCTM_User, time=2, value=0.1}
          {arriveTangent=-0.0405, interpMode=RCIM_Cubic, leaveTangent=-0.0405, tangentMode=RCTM_User, time=3, value=1}
          <at 4 0>
         ]}}}
      CurveIndex=$Emitter.Age
     )
    )
    Scale RGB=&MakeVectorFromLinearColorRGB(Color=&ColorFromCurve(ColorCurve=DataInterface() {propertyValues {
        AlphaCurve {keys [<at 0 1 RCIM_Cubic>,<at 1 1 RCIM_Cubic>]}
        BlueCurve {keys [<at 0.4624 5 RCIM_Cubic>,<at 0.7634 1 RCIM_Cubic>]}
        GreenCurve {keys [<at 0.4624 2.5 RCIM_Cubic>,<at 0.7634 1 RCIM_Cubic>]}
        RedCurve {keys [<at 0.4624 2.6431 RCIM_Cubic>,<at 0.7634 1 RCIM_Cubic>]}
       }}))
    ScaleA=true
    ScaleRGB=true
   )
   &CameraOffset(Camera Offset Amount=-10)
  ]
  rendererClasses [&NiagaraRibbonRendererProperties]
  renderers [
   {rendererClass=&NiagaraRibbonRendererProperties}
   {rendererClass=&NiagaraRibbonRendererProperties, rendererIndex=1}
   {rendererClass=&NiagaraRibbonRendererProperties, rendererIndex=2}
  ]
 }
 SmokeStreams_Upshort {
  emitterUpdateScript [
   &EmitterState(Life Cycle Mode=::NewEnumerator1, Loop Behavior=::NewEnumerator1, Loop Duration Mode=::NewEnumerator1, UseLoopDelay=true)
   &SpawnRate(SpawnRate=50, Use Spawn Probability=false) {enabled=false}
   &FN_SpawnRibbon_v2(SpawnRate=200)
  ]
  particleSpawnScript [
   &InitializeParticle(Color Mode=::NewEnumerator0, Lifetime=0.1, UsePositionOffset=false, Write Material Random=true) {enabled=false}
   &InitializeRibbon(
    Color=&MultiplyLinearColor(
     Linear Color=<rgba 0.9167 0.5 1 0.1>
     Scale Factor=&MakeVector4_FromFloat(Float=&FloatFromCurve(
       FloatCurve=DataInterface() {propertyValues {Curve {keys [{interpMode=RCIM_Cubic, tangentMode=RCTM_User, time=1.5, value=0.2}, {interpMode=RCIM_Cubic, tangentMode=RCTM_User, time=3, value=1}]}}}
       CurveIndex=$Emitter.Age
      ))
    )
    Lifetime=&FloatFromCurve(
     FloatCurve=DataInterface() {propertyValues {Curve {keys [{interpMode=RCIM_Cubic, tangentMode=RCTM_User, time=0, value=1}, {interpMode=RCIM_Cubic, tangentMode=RCTM_User, time=0.5, value=0.5}]}}}
     CurveIndex=$Emitter.Age
     Scale Curve=5
    )
    Ribbon Width=&FloatFromCurve(
     FloatCurve=DataInterface() {propertyValues {Curve {keys [{interpMode=RCIM_Cubic, tangentMode=RCTM_User, time=1.5, value=0.2}, {interpMode=RCIM_Cubic, tangentMode=RCTM_User, time=3, value=1}]}}}
     CurveIndex=$Emitter.Age
     Scale Curve=12
    )
    UsePositionOffset=false
    Write Material Random=true
   )
   &AddVelocity()
   &SkeletalMeshLocation(
    Bone / Socket Index=5
    Bone Sampling Mode=::NewEnumerator5
    MaskSpawnGroup=false
    Sampled Position Offset=<0 0 -10>
    Sampled Position Offset Coordinate Space:ENiagaraSkelMeshTransforms=::NewEnumerator2
    Skeletal Mesh=$User.SkeletalMesh
   )
   &ApplyInitialForces()
  ]
  particleUpdateScript [
   &ParticleState()
   /SetVariables(
    Particles.Position001:NiagaraPosition=&ConvertVectorToPosition(
     Input Position=&Add_Vector(
      A=&ConvertPositionToVector(Input Position=$Particles.Position, Convert To Absolute Large World Vector=false)
      B=<15 0 2>
     )
     Vector Is In Absolute Large World Space:NiagaraBool=false
    )
    Particles.Position002:NiagaraPosition=&ConvertVectorToPosition(
     Input Position=&Add_Vector(
      A=&ConvertPositionToVector(Input Position=$Particles.Position, Convert To Absolute Large World Vector=false)
      B=<-13 0 -3>
     )
     Vector Is In Absolute Large World Space:NiagaraBool=false
    )
    Particles.RibbonWidth001:NiagaraFloat=&Multiply_Float(A=$Particles.RibbonWidth, B=0.7)
    Particles.RibbonWidth002:NiagaraFloat=&Multiply_Float(A=$Particles.RibbonWidth, B=0.6)
   ) {isSetParameters=true, module=SetVariables_#5}
   &SolveForcesAndVelocity(Clamp Velocity=false, Limit Acceleration=false)
   &SkeletalMeshLocation(
    Bone / Socket Index=&SelectIntFromArray(Direct Array Index=3, Int Selection Array=DataInterface() {propertyValues {IntData [0,1]}})
    MaskSpawnGroup=false
    Mesh Sampling Type=::NewEnumerator5
    Position Sampling=::NewEnumerator1
    Skeletal Mesh=$User.SkeletalMesh
    Socket Sampling Mode=::NewEnumerator5
   ) {enabled=false, module=SkeletalMeshLocation001}
   &CurlNoiseForce(Pan Noise=false) {enabled=false, module=CurlNoiseForce001}
   &DynamicMaterialParameters(
    Index 0 Param 1=1
    Index 0 Param 2=1
    Index 0 Param 3=1
    Index 0 Param 4=1
    Index 1 Param 1=1
    Index 1 Param 2=1
    Index 2 Param 1=1
    Index 2 Param 2=1
    Index 2 Param 3=&FloatFromCurve(
     FloatCurve=DataInterface() {propertyValues {Curve {keys [
         {interpMode=RCIM_Cubic, tangentMode=RCTM_User, time=0, value=1}
         {interpMode=RCIM_Cubic, tangentMode=RCTM_User, time=0.1922, value=1}
         {tangentMode=RCTM_User, time=1, value=-1}
        ]}}}
     Scale Curve=0.5
    )
    Index 2 Param 4=&FloatFromCurve(FloatCurve=DataInterface() {propertyValues {Curve {keys [{interpMode=RCIM_Cubic, tangentMode=RCTM_User, time=0, value=2}, {interpMode=RCIM_Cubic, tangentMode=RCTM_User, time=0.5, value=1}]}}})
    Module.Param0WriteEnabled=true
    Module.Param0WriteEnabled1=true
    Module.Param0WriteEnabled2=true
    Module.Param0WriteEnabled3=true
    Module.Param1WriteEnabled=true
    Module.Param1WriteEnabled1=true
    Module.Param1WriteEnabled2=true
    Module.Param1WriteEnabled3=true
    Module.Param2WriteEnabled=true
    Module.Param2WriteEnabled1=true
    Module.Param2WriteEnabled2=true
    Module.Param2WriteEnabled3=true
    Module.Param3WriteEnabled=true
    Module.Param3WriteEnabled1=true
    Module.Param3WriteEnabled2=true
    Module.Param3WriteEnabled3=true
    Write Parameter Index 1=true
    Write Parameter Index 2=true
    Write Parameter Index 3=true
   ) {module=DynamicMaterialParameters001}
   &ScaleRibbonWidth(Ribbon Width Scale=&Multiply_Float(
     A=&Multiply_Float(
      A=&FloatFromCurve(FloatCurve=DataInterface() {propertyValues {Curve {keys [
           {interpMode=RCIM_Cubic, tangentMode=RCTM_User, time=0.1579, value=3.847}
           {arriveTangent=3.7274, interpMode=RCIM_Cubic, leaveTangent=3.7274, time=0.3145, value=5.967}
           {arriveTangent=-7.9244, interpMode=RCIM_Cubic, leaveTangent=-7.9244, time=0.4262, value=4.847}
           {arriveTangent=-4.9452, interpMode=RCIM_Cubic, leaveTangent=-4.9452, time=0.6889, value=3}
           {arriveTangent=0.6917, interpMode=RCIM_Cubic, leaveTangent=0.6917, tangentMode=RCTM_User, time=0.981, value=2.1033}
          ]}}})
      B=1
     )
     B=&FloatFromCurve(
      FloatCurve=DataInterface() {propertyValues {Curve {keys [
          {interpMode=RCIM_Cubic, tangentMode=RCTM_User, time=2, value=0.8}
          {interpMode=RCIM_Cubic, tangentMode=RCTM_User, time=3, value=0.8}
          <at 5 1 RCIM_Cubic>
         ]}}}
      CurveIndex=$Emitter.Age
     )
    ))
   &CurlNoiseForce(
    Noise Frequency=5
    Noise Strength=&FloatFromCurve(
     FloatCurve=DataInterface() {propertyValues {Curve {keys [{arriveTangent=-3.2172, interpMode=RCIM_Cubic, leaveTangent=-3.2172, tangentMode=RCTM_User, time=-0.0184, value=1.5941}, <at 1.0003 0.7389>]}}}
     CurveIndex=$Emitter.Age
     Scale Curve=1500
    )
    Pan Noise=true
    Pan Noise Field=<0 0 2>
    Random Seed:NiagaraInt32=$Particles.SpawnGroup
   ) {enabled=false}
   &DynamicMaterialParameters(
    Index 0 Param 1=1
    Index 0 Param 2=1
    Index 0 Param 3=1
    Index 0 Param 4=1
    Index 1 Param 1=1
    Index 1 Param 2=1
    Index 2 Param 1=0.2
    Index 2 Param 2=0.2
    Index 2 Param 3=0.2
    Index 2 Param 4=1
    Module.Param0WriteEnabled=true
    Module.Param0WriteEnabled1=true
    Module.Param0WriteEnabled2=true
    Module.Param1WriteEnabled=true
    Module.Param1WriteEnabled1=true
    Module.Param1WriteEnabled2=true
    Module.Param2WriteEnabled=true
    Module.Param2WriteEnabled1=true
    Module.Param2WriteEnabled2=true
    Module.Param3WriteEnabled=true
    Module.Param3WriteEnabled1=true
    Module.Param3WriteEnabled2=true
    Write Parameter Index 1=true
    Write Parameter Index 2=true
   ) {enabled=false}
   &ScaleColor(
    Scale Alpha=&Multiply_Float(
     A=&FloatFromCurve(FloatCurve=DataInterface() {propertyValues {Curve {keys [
          <at 0 0 RCIM_Cubic>
          {arriveTangent=2, interpMode=RCIM_Cubic, leaveTangent=2, time=0.3, value=1}
          {interpMode=RCIM_Cubic, tangentMode=RCTM_User, time=0.5, value=1}
          {interpMode=RCIM_Cubic, tangentMode=RCTM_User, time=1, value=0}
         ]}}})
     B=&FloatFromCurve(
      FloatCurve=DataInterface() {propertyValues {Curve {keys [
          {arriveTangent=0.1443, interpMode=RCIM_Cubic, leaveTangent=0.1443, tangentMode=RCTM_User, time=1, value=0}
          {arriveTangent=0.6557, interpMode=RCIM_Cubic, leaveTangent=0.6557, tangentMode=RCTM_User, time=2, value=0.1}
          {arriveTangent=-0.0405, interpMode=RCIM_Cubic, leaveTangent=-0.0405, tangentMode=RCTM_User, time=3, value=1}
          <at 4 0 RCIM_Cubic>
         ]}}}
      CurveIndex=$Emitter.Age
     )
    )
    Scale RGB=&MakeVectorFromLinearColorRGB(Color=&ColorFromCurve(ColorCurve=DataInterface() {propertyValues {
        AlphaCurve {keys [<at 0 1 RCIM_Cubic>,<at 1 1 RCIM_Cubic>]}
        BlueCurve {keys [<at 0.4624 5 RCIM_Cubic>,<at 0.7634 1 RCIM_Cubic>]}
        GreenCurve {keys [<at 0.4624 2.5 RCIM_Cubic>,<at 0.7634 1 RCIM_Cubic>]}
        RedCurve {keys [<at 0.4624 2.6431 RCIM_Cubic>,<at 0.7634 1 RCIM_Cubic>]}
       }}))
    ScaleA=true
    ScaleRGB=true
   )
   &CameraOffset(Camera Offset Amount=-15)
  ]
  rendererClasses [&NiagaraRibbonRendererProperties]
  renderers [
   {rendererClass=&NiagaraRibbonRendererProperties}
   {rendererClass=&NiagaraRibbonRendererProperties, rendererIndex=1}
   {rendererClass=&NiagaraRibbonRendererProperties, rendererIndex=2}
  ]
 }
 Surround_modified {
  emitterUpdateScript [
   &EmitterState(
    Inactive Response=::NewEnumerator1
    Life Cycle Mode=::NewEnumerator1
    Loop Behavior=::NewEnumerator1
    Loop Delay=2.2
    Loop Duration=0.5
    UseLoopDelay=true
   )
   &SpawnRate(SpawnRate=20, Use Spawn Probability=false)
  ]
  particleSpawnScript [
   /SetVariables(Particles.Random=&RandomRangeFloat(Minimum=0, Maximum=1)) {isSetParameters=true, module=SetVariables_#4}
   /Niagara/Modules/Spawn/Initialization/InitializeParticle(
    Color=<rgba 1 0 0 1>
    Lifetime=&Lerp_Float(A=0.3, B=0.5, Alpha=$Particles.Random)
    Module.Write Color=true
    Module.Write Lifetime=true
    Module.Write Mass=true
    Module.Write Material Random=true
    Module.Write Position=false
    Module.Write Scale=false
    Module.Write SpriteRotation=false
    Module.Write SpriteSize=true
    Position=&SimulationPosition()
    Sprite Size=&Vector2DFromCurve(
     Vector2Curve=DataInterface() {propertyValues {
       XCurve {keys [
         {interpMode=RCIM_Cubic, tangentMode=RCTM_User, time=0, value=1}
         {arriveTangent=1.1279, interpMode=RCIM_Cubic, leaveTangent=1.1279, tangentMode=RCTM_User, time=1, value=2}
        ]}
       YCurve {keys [{interpMode=RCIM_Cubic, tangentMode=RCTM_User, time=0, value=0}, {interpMode=RCIM_Cubic, tangentMode=RCTM_User, time=1, value=1}]}
      }}
     CurveIndex=$Emitter.Age
     ScaleCurve=<20 80>
    )
   )
   &CameraOffset(Camera Offset Amount=&RandomRangeFloat(Minimum=0, Maximum=5))
   &DynamicMaterialParameters(
    Index 0 Param 1=&Lerp_Float(A=0, B=0.2, Alpha=$Particles.Random)
    Index 0 Param 2=&Lerp_Float(A=0, B=1, Alpha=$Particles.Random)
    Index 0 Param 3=20
    Index 0 Param 4=2
    Module.Param0WriteEnabled=true
    Module.Param1WriteEnabled=true
    Module.Param2WriteEnabled=true
    Module.Param3WriteEnabled=true
   )
   &CylinderLocation(
    Cylinder Height=1
    Cylinder Radius=0.1
    MaskSpawnGroup=false
    Offset=<0 30 120>
    Offset Coordinate Space:ENiagaraCoordinateSpace=::Local
    Offset Mode:ENiagara_OffsetMode=::NewEnumerator0
    Override Seed=false
    UseOwnerScale=true
   )
   &AddVelocity(Coordinate Space=::World, Velocity=<0 0 20>)
  ]
  particleUpdateScript [
   &ParticleState()
   &ScaleSpriteSize(Scale Factor=&MakeVector2D(
     X=&FloatFromCurve(FloatCurve=DataInterface() {propertyValues {Curve {keys [<at 0 1.5>,<at 1 0.5>]}}})
     Y=&FloatFromCurve(FloatCurve=DataInterface() {propertyValues {Curve {keys [<at 0 1>,<at 1 2>]}}})
    ))
   &SolveForcesAndVelocity(Clamp Velocity=false, Limit Acceleration=false)
   &ScaleColor(
    Scale Mode=::NewEnumerator1
    Scale RGBA=&MakeVector4_FromLinearColor(LinearColor=&HueShiftLinearColor(Hue=0, Saturation=0, Value=1, Linear Color=$Particles.Color))
    ScaleRGBA=true
   )
  ]
  rendererClasses [&NiagaraSpriteRendererProperties]
  renderers [{rendererClass=&NiagaraSpriteRendererProperties}]
 }
}
summary {
 systemName=NS_Emote_Origin_Prison3
 emitters [
  {emitterName=Hands_Pulse, rendererClasses [&NiagaraSpriteRendererProperties]}
  {emitterName=SmokeStreams_Up, rendererClasses [&NiagaraRibbonRendererProperties]}
  {emitterName=SmokeStreams_Upshort, rendererClasses [&NiagaraRibbonRendererProperties]}
  {emitterName=Surround_modified, rendererClasses [&NiagaraSpriteRendererProperties]}
  {emitterName=Blocks, rendererClasses [&NiagaraSpriteRendererProperties]}
 ]
 userVariables [
  User.bIsFrontEnd {type=NiagaraBool}
  User.bIsFrontEndPreview {type=NiagaraBool, defaultValue=true}
  User.Color_01 {type=LinearColor, description="", defaultValue {struct=LinearColor, value=<rgba 1 1 1 1>}}
  User.Color_02 {type=LinearColor, description="", defaultValue {struct=LinearColor, value=<rgba 1 1 1 1>}}
  User.ColorA {type=LinearColor, description="", defaultValue {struct=LinearColor, value=<rgba 0.4938 0 2 1>}}
  User.ColorB {type=LinearColor, description="", defaultValue {struct=LinearColor, value=<rgba 1.25 0.3125 0 1>}}
  User.HelmetOff {type=NiagaraFloat, defaultValue=1, description=""}
  User.HelmetOff_ {type=NiagaraBool, defaultValue=true}
  User.IsSwing {type=NiagaraBool}
  User.ParaGlideLeanAlpha {type=NiagaraFloat, defaultValue=0, description=""}
  User.RibbonIDOffset {type=NiagaraInt32}
  User.SkeletalMesh {
   type=NiagaraDataInterfaceSkeletalMesh
   defaultValue {
    struct=NiagaraExt_VariableValue_DataInterface
    value {dataInterface=&self:NiagaraDataInterfaceSkeletalMesh_0}
   }
  }
 ]
}
