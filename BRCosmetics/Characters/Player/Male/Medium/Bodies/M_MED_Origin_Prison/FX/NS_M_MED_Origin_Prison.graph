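# Niagara cosmetic effect for the Origin Prison character body, attaching glowing energy to both hands and the head via skeletal mesh sampling.
# Energy draws layered meshes with flickering lights, Cube_Trails_Small emits color-shifting mesh trails, and Decals pins a sprite to the head.
# Spawn groups select each bone's offsets, while system hand and head values gate spawning, scale or visibility.
# BigHeadMode enlarges head effects, and smoothed positions feed the energy materials.
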NiagaraSystem /BRCosmetics/Characters/Player/Male/Medium/Bodies/M_MED_Origin_Prison/FX/NS_M_MED_Origin_Prison
dependencies {
 usedDataInterfaces [
  &NiagaraDataInterfaceMeshRendererInfo
  &NiagaraDataInterfaceCurve
  &NiagaraDataInterfacePlatformSet
  &NiagaraDataInterfaceSkeletalMesh
  &NiagaraDataInterfaceArrayFloat3
 ]
 usedDynamicInputs [
  &NormalizeVector
  &Boolean_OrOperation
  &SetBoolByFloatComparison
 ]
 usedModules [
  &self:SystemSpawnScript.SystemScriptSource.SystemScriptGraph.NiagaraNodeAssignment_3.SetVariables_#1
  &SystemState
  &self:SystemSpawnScript.SystemScriptSource.SystemScriptGraph.NiagaraNodeAssignment_5.SetVariables_#2
  &self:SystemSpawnScript.SystemScriptSource.SystemScriptGraph.NiagaraNodeAssignment_1.SetVariables_#3
  &self:SystemSpawnScript.SystemScriptSource.SystemScriptGraph.NiagaraNodeAssignment_0.SetVariables_#4
  &Timeline
  &self:SystemSpawnScript.SystemScriptSource.SystemScriptGraph.NiagaraNodeAssignment_2.SetVariables_#5
 ]
}
emitters {
 Cube_Trails_Small {
  emitterUpdateScript [
   &EmitterState()
   &SpawnRate(Spawn Probability=$System.Hand Left, SpawnRate=20, Use Spawn Probability=true)
   &SpawnRate(Spawn Probability=$System.Hand Right, SpawnGroup=1, SpawnRate=20, Use Spawn Probability=true) {module=SpawnRate002}
   &SpawnRate(Spawn Probability=$System.Head, SpawnGroup=2, SpawnRate=30, Use Spawn Probability=true) {module=SpawnRate001}
  ]
  particleSpawnScript [
   /SetVariables(Particles.SpawnGroup=$Emitter.SpawnGroup) {isSetParameters=true}
   &InitializeParticle(
    Color=<rgba 2 2 2 1>
    Lifetime Max=1
    Lifetime Min=0.75
    Lifetime Mode=::NewEnumerator1
    Mesh Scale=&Multiply_VectorByFloat(
     Vector=&MakeVector(
      X=&RandomRangeFloat(Minimum=1.5, Maximum=3)
      Y=&RandomRangeFloat(Minimum=0.5, Maximum=1)
      Z=&RandomRangeFloat(Minimum=0.5, Maximum=1)
     )
     Float=&Lerp_Float(
      A=&Lerp_Float(A=1, B=2, Alpha=$NPC.FortniteNPC.BigHeadMode)
      B=1
      Alpha=&MaskFloatBySpawnGroup(Masked Float=0, Passthrough Float=1, Spawn Group Mask=2)
     )
    )
    Mesh Scale Mode=::NewEnumerator4
    UsePositionOffset=false
    Write Material Random=true
   )
   &InitialMeshOrientation(Mesh Orientation Mode=::NewEnumerator2, Rotation Coordinate Space=::NewEnumerator2, Use Rotation Vector=false)
  ]
  particleUpdateScript [
   &ParticleState()
   /SetVariables(Particles.VisibilityTag=&MakeIntFromFloat(Float=&OneMinusFloat(Float=&self:Dynamic_Float_Array(
       Index:NiagaraInt32=&SpawnGroup()
       Direct Array Mode:ENiagara_WrapClamp=::NewEnumerator0
       "Index [0]":NiagaraFloat=$System.Hand Left
       "Index [1]":NiagaraFloat=$System.Hand Right
       "Index [2]":NiagaraFloat=$System.Head
       "Index [3]":NiagaraFloat=1
      )))) {isSetParameters=true}
   &SkeletalMeshLocation(
    Bone / Socket Index=$Particles.SpawnGroup
    Bone Sampling Mode=::NewEnumerator4
    MaskSpawnGroup=false
    Orientation Sampling=::NewEnumerator1
    Sampled Position Offset=&Multiply_VectorByFloat(
     Vector=&SelectVectorFromArray(Direct Array Index=$Particles.SpawnGroup, Vector Selection Array=$System.PositionOffset)
     Float=&Lerp_Float(
      A=&Lerp_Float(A=1, B=2, Alpha=$NPC.FortniteNPC.BigHeadMode)
      B=1
      Alpha=&MaskFloatBySpawnGroup(Masked Float=0, Passthrough Float=1, Spawn Group Mask=2)
     )
    )
    Skeletal Mesh=$System.Skeletal Mesh
   )
   /SetVariables(Particles.Position=&Add_Vector(
     A=&ConvertPositionToVector(Input Position=$Particles.Position, Convert To Absolute Large World Vector=false)
     B=&Multiply_VectorByFloat(
      Vector=&Add_Vector(
       A=&Lerp_Vector(A=<0 0 1>, B=$System.DiveVelocityScale, Alpha=&Power(a=$Particles.NormalizedAge, b=3))
       B=&RotateVectorByQuaternion(
        Vector To Rotate=&SelectVectorFromArray(Direct Array Index=$Particles.SpawnGroup, Vector Selection Array=$System.AngleOffset)
        Quaternion=$Particles.SkeletalMeshLocation.SampledRotation
       )
      )
      Float=&FloatFromCurve(
       FloatCurve=DataInterface() {propertyValues {Curve {keys [
           {arriveTangent=0.5799, interpMode=RCIM_Cubic, leaveTangent=0.5799, tangentMode=RCTM_User, time=0, value=-0.1}
           {arriveTangent=1.4467, interpMode=RCIM_Cubic, leaveTangent=1.4467, tangentMode=RCTM_User, time=1, value=1}
          ]}}}
       Scale Curve=&Multiply_Float(
        A=&MaskFloatBySpawnGroup(Masked Float=20, Passthrough Float=15, Spawn Group Mask=2)
        B=&RandomRangeFloat(Minimum=0.5, Maximum=1)
       )
      )
     )
    )) {isSetParameters=true}
   &RotateAroundPoint(
    Derive the Rotation Axis=false
    Radius:NiagaraFloat=&Multiply_Float(
     A=&Multiply_Float(
      A=&Lerp_Float(
       A=&RandomRangeFloat(Minimum=1, Maximum=1.5)
       B=&RandomRangeFloat(Minimum=3, Maximum=6)
       Alpha=&FloatFromCurve(FloatCurve=DataInterface() {propertyValues {Curve {keys [
            {arriveTangent=3.8353, interpMode=RCIM_Cubic, leaveTangent=3.8353, tangentMode=RCTM_User, time=0, value=0.5}
            {arriveTangent=-0.5, interpMode=RCIM_Cubic, leaveTangent=-0.5, time=0.3, value=0.9847}
            {interpMode=RCIM_Cubic, leaveTangent=-1.7958, time=1, value=0}
           ]}}})
      )
      B=&MaskFloatBySpawnGroup(Masked Float=1.15, Passthrough Float=0.7, Spawn Group Mask=2)
     )
     B=&Lerp_Float(
      A=&Lerp_Float(A=1, B=1.75, Alpha=$NPC.FortniteNPC.BigHeadMode)
      B=1
      Alpha=&MaskFloatBySpawnGroup(Masked Float=0, Passthrough Float=1, Spawn Group Mask=2)
     )
    )
    Rotation Axis:Vector3f=$System.DiveVelocityScale
    Rotation Center:NiagaraPosition=&Lerp_Vector(
     A=<0 0 0>
     B=&Lerp_Vector(
      A=&Lerp_Vector(A=<0 0 0>, B=<0 0 7>, Alpha=$NPC.FortniteNPC.BigHeadMode)
      B=<0 0 0>
      Alpha=&MaskFloatBySpawnGroup(Masked Float=0, Passthrough Float=1, Spawn Group Mask=2)
     )
     Alpha=$NPC.FortniteNPC.BigHeadMode
    )
    Rotation Phase:NiagaraFloat=&RandomRangeFloat(Minimum=0, Maximum=1)
    Rotation Rate:NiagaraFloat=&MakeCustomFloatFromBool(Bool=&RandomBool(), True Float=0.6, False Float=-0.6)
    Set Particle Position=false
   )
   /SetVariables(Particles.Rotational_Position:NiagaraPosition=&ConvertVectorToPosition(
     Input Position=&Add_Vector(
      A=&ConvertPositionToVector(Input Position=$Particles.Position, Convert To Absolute Large World Vector=false)
      B=Linked() {linkedVariable {name=Output.RotateAroundPoint.OutputPosition, type=NiagaraPosition}}
     )
     Vector Is In Absolute Large World Space:NiagaraBool=true
    )) {isSetParameters=true}
   &UpdateMeshOrientation(
    Facing Direction=&Add_Vector(
     A=$System.DiveVelocityScale
     B=&RotateVectorByQuaternion(
      Vector To Rotate=&SelectVectorFromArray(Direct Array Index=$Particles.SpawnGroup, Vector Selection Array=$System.AngleOffset)
      Quaternion=$Particles.SkeletalMeshLocation.SampledRotation
     )
    )
    Facing Interpolation=false
    Orientation Method=::NewEnumerator1
    Override Quat=false
    Side Direction:Vector3f=$Particles.RotateAroundPoint.VectorFromCenter
    Stabilization Mode:ENiagara_OrientVectorCount=::NewEnumerator2
    TransformFacing=false
    TransformSide=false
   )
   &ScaleMeshSize(Scale Factor=&Multiply_VectorByFloat(
     Vector=&Lerp_Vector(
      A=<0.5 1 1>
      B=<0.75 1 1>
      Alpha=&FloatFromCurve(FloatCurve=DataInterface() {propertyValues {Curve {keys [
           {arriveTangent=0.9666, interpMode=RCIM_Cubic, time=0.2, value=0}
           {interpMode=RCIM_Cubic, tangentMode=RCTM_User, time=0.5, value=1}
           {interpMode=RCIM_Cubic, tangentMode=RCTM_User, time=0.8, value=1}
           <at 1 0 RCIM_Cubic>
          ]}}})
     )
     Float=&FloatFromCurve(
      FloatCurve=DataInterface() {propertyValues {Curve {keys [
          {arriveTangent=3.3011, interpMode=RCIM_Cubic, leaveTangent=3.3011, tangentMode=RCTM_User, time=0, value=0.25}
          {interpMode=RCIM_Cubic, tangentMode=RCTM_User, time=0.15, value=1}
          {interpMode=RCIM_Cubic, tangentMode=RCTM_Break, time=0.5, value=1}
          {arriveTangent=-3.8115, interpMode=RCIM_Constant, leaveTangent=-3.8115, time=1, value=0.25}
         ]}}}
      Scale Curve=0.75
     )
    ))
   &DynamicMaterialParameters(
    Index 0 Param 1=&FloatFromCurve(
     FloatCurve=DataInterface() {propertyValues {Curve {keys [{interpMode=RCIM_Cubic, tangentMode=RCTM_User, time=0.6, value=0}, {interpMode=RCIM_Cubic, tangentMode=RCTM_User, time=1, value=1}]}}}
     Scale Curve=6
    )
    Module.Param0WriteEnabled=true
    Module.Param1WriteEnabled=false
    Module.Param2WriteEnabled=false
    Module.Param3WriteEnabled=false
   )
   &Color(Color=&LerpLinearColors(
     StartColor=$User.ColorA
     EndColor=$User.ColorB
     LerpFactor=&FloatFromCurve(FloatCurve=DataInterface() {propertyValues {Curve {keys [{interpMode=RCIM_Cubic, tangentMode=RCTM_User, time=0.6, value=0}, {interpMode=RCIM_Cubic, leaveTangent=5.1459, time=0.75, value=1}]}}})
    ))
   &CameraOffset(Camera Offset Amount=&MaskFloatBySpawnGroup(Masked Float=0.5, Passthrough Float=0.5, Spawn Group Mask=2))
  ]
  rendererClasses [&NiagaraMeshRendererProperties]
  renderers [{rendererClass=&NiagaraMeshRendererProperties, rendererIndex=0}]
 }
 Decals {
  emitterSpawnScript [/SetVariables() {isSetParameters=true}]
  emitterUpdateScript [
   &EmitterState(
    Inactive Response=::NewEnumerator1
    Life Cycle Mode=::NewEnumerator1
    Loop Behavior=::NewEnumerator1
    Loop Duration Mode=::NewEnumerator1
    UseLoopDelay=false
   )
   &SpawnBurst_Instantaneous(Spawn Probability=$System.Head, Use Spawn Probability=true, UseLoopCountLimit=false)
  ]
  particleSpawnScript [
   &InitializeParticle(Color=<rgba 15 4.1613 0 1>, Sprite Size Mode=::NewEnumerator3, Uniform Sprite Size=20, UsePositionOffset=true, Write Material Random=true)
   &DynamicMaterialParameters(
    Index 0 Param 1=1
    Index 0 Param 2=1
    Index 1 Param 1=15
    Index 1 Param 2=30
    Module.Param0WriteEnabled=true
    Module.Param0WriteEnabled1=true
    Module.Param1WriteEnabled=true
    Module.Param1WriteEnabled1=true
    Module.Param2WriteEnabled=true
    Module.Param2WriteEnabled1=true
    Module.Param3WriteEnabled=true
    Module.Param3WriteEnabled1=true
    Write Parameter Index 1=true
   )
   &CameraOffset(Camera Offset Amount=10)
  ]
  particleUpdateScript [
   &ParticleState(Kill Particles When Lifetime Has Elapsed=false, Let Infinitely Lived Particles Die When Emitter Deactivates=true)
   &ScaleSpriteSize(
    Scale Sprite Size Mode=::NewEnumerator0
    Uniform Scale Factor=&Lerp_Float(A=1, B=2, Alpha=$NPC.FortniteNPC.BigHeadMode)
   )
   /SetVariables(Particles.VisibilityTag=&MakeIntFromFloat(Float=&OneMinusFloat(Float=$System.Head))) {isSetParameters=true}
   &SkeletalMeshLocation(
    Bone / Socket Index=2
    Bone Sampling Mode=::NewEnumerator4
    MaskSpawnGroup=false
    Orientation Sampling=::NewEnumerator1
    Sampled Position Offset=&Lerp_Vector(A=<10 -2 0>, B=<20 -4 0>, Alpha=$NPC.FortniteNPC.BigHeadMode)
    Skeletal Mesh=$System.Skeletal Mesh
   )
  ]
  rendererClasses [&NiagaraSpriteRendererProperties]
  renderers [{rendererClass=&NiagaraSpriteRendererProperties}]
 }
 Energy {
  emitterUpdateScript [
   &EmitterState(
    Inactive Response=::NewEnumerator1
    Life Cycle Mode=::NewEnumerator1
    Loop Behavior=::NewEnumerator1
    Loop Duration Mode=::NewEnumerator1
    UseLoopDelay=false
   )
   &SpawnBurst_Instantaneous(Use Spawn Probability=false, UseLoopCountLimit=false)
   &SpawnBurst_Instantaneous(Spawn Group=1, Use Spawn Probability=false, UseLoopCountLimit=false) {module=SpawnBurst_Instantaneous001}
   &SpawnBurst_Instantaneous(Spawn Group=2, Use Spawn Probability=false, UseLoopCountLimit=false) {module=SpawnBurst_Instantaneous002}
  ]
  particleSpawnScript [
   /SetVariables(Particles.SpawnGroup=$Emitter.SpawnGroup) {isSetParameters=true}
   &InitializeParticle(
    Lifetime=2
    Mesh Scale=&Lerp_Vector(
     A=<0.425 0.425 0.2>
     B=<0.35 0.35 0.15>
     Alpha=&MaskFloatBySpawnGroup(Masked Float=0, Passthrough Float=1, Spawn Group Mask=2)
    )
    Mesh Scale Mode=::NewEnumerator4
    UsePositionOffset=false
    Write Material Random=true
   )
   &InitialMeshOrientation(
    Mesh Axis To Orient=<0 0 1>
    Mesh Orientation Mode=::NewEnumerator2
    Rotation=&Lerp_Vector(A=<0 0 0>, B=<0.5 0 0>, Alpha=&MakeFloatFromInt(Int=$Particles.SpawnGroup))
    Rotation Coordinate Space=::NewEnumerator0
    Use Rotation Vector=true
   )
   &DynamicMaterialParameters(
    Index 0 Param 2=&RandomRangeFloat(Minimum=0, Maximum=1.5)
    Index 0 Param 4=1
    Index 2 Param 2=20
    Index 2 Param 4=1
    Module.Param0WriteEnabled=true
    Module.Param0WriteEnabled2=false
    Module.Param1WriteEnabled=true
    Module.Param1WriteEnabled2=true
    Module.Param2WriteEnabled=true
    Module.Param2WriteEnabled2=true
    Module.Param3WriteEnabled=true
    Module.Param3WriteEnabled2=true
    Write Parameter Index 2=true
   )
   /SetVariables(Particles.LightRadius=40) {isSetParameters=true}
  ]
  particleUpdateScript [
   &ParticleState(
    Kill Particles When Lifetime Has Elapsed=false
    Let Infinitely Lived Particles Die When Emitter Deactivates=true
    Loop Particles Lifetime=true
   )
   &ScaleMeshSize(Scale Factor=&Lerp_Vector(
     A=&Lerp_Vector(A=<1 1 1>, B=<1.7 1.7 1.7>, Alpha=$NPC.FortniteNPC.BigHeadMode)
     B=<1 1 1>
     Alpha=&MaskFloatBySpawnGroup(Masked Float=0, Passthrough Float=1, Spawn Group Mask=2)
    )) {module=ScaleMeshSize001}
   &SkeletalMeshLocation(
    Bone / Socket Index=$Particles.SpawnGroup
    Bone Sampling Mode=::NewEnumerator4
    MaskSpawnGroup=false
    Orientation Sampling=::NewEnumerator1
    Sampled Position Offset=&Multiply_VectorByFloat(
     Vector=&AddVectorToPosition(
      Position=&SelectVectorFromArray(Direct Array Index=$Particles.SpawnGroup, Vector Selection Array=$System.PositionOffset)
      Vector=&Lerp_Vector(
       A=&Lerp_Vector(A=<1 1 1>, B=<2 2 2>, Alpha=$NPC.FortniteNPC.BigHeadMode)
       B=<0 0 0>
       Alpha=&MaskFloatBySpawnGroup(Masked Float=0, Passthrough Float=1, Spawn Group Mask=2)
      )
     )
     Float=&Lerp_Float(
      A=&Lerp_Float(A=1, B=2, Alpha=$NPC.FortniteNPC.BigHeadMode)
      B=1
      Alpha=&MaskFloatBySpawnGroup(Masked Float=0, Passthrough Float=1, Spawn Group Mask=2)
     )
    )
    Skeletal Mesh=$System.Skeletal Mesh
   )
   &UpdateMeshOrientation(
    Facing Direction=&Add_Vector(
     A=<0 0 1>
     B=&RotateVectorByQuaternion(
      Vector To Rotate=&SelectVectorFromArray(Direct Array Index=$Particles.SpawnGroup, Vector Selection Array=$System.AngleOffset)
      Quaternion=$Particles.SkeletalMeshLocation.SampledRotation
     )
    )
    Facing Interpolation=false
    Orientation Method=::NewEnumerator1
    Override Quat=false
    TransformFacing=false
    TransformUp=false
   ) {module=UpdateMeshOrientation002}
   &UpdateMeshOrientation(Rotation Rate=-1, Rotation Vector=<0 0 1>)
   &ScaleColor(
    Scale Alpha=0.6
    Scale RGB=&Multiply_VectorByFloat(Vector=&MakeVectorFromLinearColorRGB(Color=$User.ColorA), Float=0.8)
    ScaleA=true
    ScaleRGB=true
   ) {module=ScaleColor001}
   /SetVariables(Particles.Color_Yellow:LinearColor=&MakeLinearColorFromVectorAndFloat(
     "Vector (RGB)"=&Multiply_VectorByFloat(Vector=&MakeVectorFromLinearColorRGB(Color=$User.ColorB), Float=6)
     "Float (Alpha)"=1
    )) {isSetParameters=true}
   &ScaleMeshSize(Scale Factor=&MaskVectorBySpawnGroup(
     Masked Vector=&Lerp_Vector(A=<0 0 0>, B=<1 1 1>, Alpha=$System.Hand Left)
     Passthrough Vector=&MaskVectorBySpawnGroup(
      Masked Vector=&Lerp_Vector(A=<0 0 0>, B=<1 1 1>, Alpha=$System.Hand Right)
      Passthrough Vector=&Lerp_Vector(A=<0 0 0>, B=<1 1 1>, Alpha=$System.Head)
      Spawn Group Mask=1
     )
    ))
   /SetVariables(Particles.Scale_Yellow:Vector3f=&Multiply_Vector(A=$Particles.Scale, B=<0.6 0.6 0.55>)) {isSetParameters=true}
   /SetVariables(Particles.LightColor=&MultiplyLinearColor(
     Linear Color=$User.ColorA
     Scale Factor=&MakeFloat4FromVectorAndFloat(
      "Vector (RGB)"=&VectorFromFloat(Value=&Lerp_Float(
        A=0.5
        B=1
        Alpha=&FloatFromCurve(FloatCurve=DataInterface() {propertyValues {Curve {keys [
             <at 0 0 RCIM_Cubic>
             {arriveTangent=1.304, interpMode=RCIM_Cubic, leaveTangent=1.304, time=0.1, value=0.75}
             {arriveTangent=-1.2483, interpMode=RCIM_Cubic, leaveTangent=-1.2483, time=0.1681, value=0.2192}
             {arriveTangent=-0.9034, interpMode=RCIM_Cubic, leaveTangent=-0.9034, time=0.216, value=1}
             {arriveTangent=-0.1111, interpMode=RCIM_Cubic, leaveTangent=-0.1111, time=0.3, value=0.1}
             {arriveTangent=1.8356, interpMode=RCIM_Cubic, leaveTangent=1.8356, time=0.3516, value=0.6401}
             {arriveTangent=2.4257, interpMode=RCIM_Cubic, leaveTangent=2.4257, time=0.4219, value=0.05}
             {arriveTangent=0.2031, interpMode=RCIM_Cubic, leaveTangent=0.2031, time=0.5, value=1}
             {arriveTangent=-2.3742, interpMode=RCIM_Cubic, leaveTangent=-2.3742, time=0.5703, value=0.3538}
             {arriveTangent=-1.957, interpMode=RCIM_Cubic, leaveTangent=-1.957, time=0.6325, value=0.6853}
             {arriveTangent=-0.2171, interpMode=RCIM_Cubic, leaveTangent=-0.2171, time=0.7, value=0}
             {arriveTangent=1.2231, interpMode=RCIM_Cubic, leaveTangent=1.2231, time=0.7714, value=0.6552}
             {arriveTangent=1.1257, interpMode=RCIM_Cubic, leaveTangent=1.1257, time=0.8336, value=0.2634}
             {arriveTangent=-1.5833, interpMode=RCIM_Cubic, leaveTangent=-1.5833, time=0.9, value=1}
             <at 1 0 RCIM_Cubic>
            ]}}})
       ))
      "Float (Alpha)"=1
     )
    )) {isSetParameters=true}
   &InterpolateOverTime(
    Rate Of Change=5
    Target Vector Value:Vector3f=&ConvertPositionToVector(
     Input Position=&ConvertVectorToPosition(
      Input Position=Linked() {linkedVariable {name=Particles.SkeletalMeshLocation.SampledPosition, type=NiagaraPosition}}
      Vector Is In Absolute Large World Space:NiagaraBool=true
     )
     Convert To Absolute Large World Vector=false
    )
    Variable Type=::NewEnumerator2
   )
   /SetVariables(Particles.InterpolateDynamic:Vector4f=&MakeFloat4FromVectorAndFloat("Vector (RGB)"=$Particles.InterpolateOverTime.InterpolateOverTime_V3_.Moving Average, "Float (Alpha)"=1.5)) {isSetParameters=true}
   /SetVariables(Particles.InterpolateDynamic_Yellow:Vector4f=&MakeFloat4FromVectorAndFloat("Vector (RGB)"=$Particles.InterpolateOverTime.InterpolateOverTime_V3_.Moving Average, "Float (Alpha)"=0.4)) {isSetParameters=true}
   &CameraOffset(Camera Offset Amount=&MaskFloatBySpawnGroup(
     Masked Float=&MakeCustomFloatFromBool(
      Bool=&IsPlatformSetActive(New Platform Set=DataInterface() {propertyValues {Platforms {
          qualityLevelMask=-32
          deviceProfileStates [
           {profileName=Mobile, qualityLevelMask=4294967295, setQualityLevelMask=4294967295}
           {profileName=Switch, qualityLevelMask=4294967295, setQualityLevelMask=4294967295}
          ]
         }}})
      True Float=0
      False Float=10
     )
     Passthrough Float=5
     Spawn Group Mask=2
    ))
   /SetVariables(Particles.CameraOffset_Yellow:NiagaraFloat=&MaskFloatBySpawnGroup(
     Masked Float=&MakeCustomFloatFromBool(
      Bool=&IsPlatformSetActive(New Platform Set=DataInterface() {propertyValues {Platforms {
          qualityLevelMask=-30
          deviceProfileStates [
           {profileName=Switch, qualityLevelMask=4294967295, setQualityLevelMask=4294967295}
           {profileName=Mobile, qualityLevelMask=4294967295, setQualityLevelMask=4294967295}
          ]
         }}})
      True Float=0
      False Float=7.5
     )
     Passthrough Float=5
     Spawn Group Mask=2
    )) {isSetParameters=true}
  ]
  rendererClasses [&NiagaraMeshRendererProperties, &NiagaraLightRendererProperties]
  renderers [
   {rendererClass=&NiagaraMeshRendererProperties, rendererIndex=0}
   {rendererClass=&NiagaraMeshRendererProperties, rendererIndex=1}
   {rendererClass=&NiagaraLightRendererProperties, rendererIndex=2}
  ]
 }
}
summary {
 systemName=NS_M_MED_Origin_Prison
 emitters [
  {emitterName=Cube_Trails_Small, rendererClasses [&NiagaraMeshRendererProperties]}
  {emitterName=Energy, rendererClasses [&NiagaraMeshRendererProperties, &NiagaraLightRendererProperties]}
  {emitterName=Decals, rendererClasses [&NiagaraSpriteRendererProperties]}
 ]
 userVariables [
  User.bIsFrontEnd {type=NiagaraBool}
  User.bIsFrontEndPreview {type=NiagaraBool, defaultValue=true}
  User.bisHandHidden {type=NiagaraBool}
  User.bisHeadHidden {type=NiagaraBool}
  User.ColorA {type=LinearColor, description="", defaultValue {struct=LinearColor, value=<rgba 0.4938 0 2 1>}}
  User.ColorB {type=LinearColor, description="", defaultValue {struct=LinearColor, value=<rgba 1.25 0.3125 0 1>}}
  User.VariantHideHead {type=NiagaraFloat, defaultValue=1, description=""}
 ]
}
